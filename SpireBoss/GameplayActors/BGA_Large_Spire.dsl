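; BGA_Large_Spire is the Spire boss actor, handling intro plunge, beam sweep, shielding, music and death loot.
; A state component and nearby-pawn sensing drive intro events and encounter manager targets, with max health scaled by playlist team size.
; Four child weakpoints take full damage, other hits are divided, and shielded hits are nullified.
; Health thresholds trigger hit reactions and damage visuals through custom primitive data, Niagara and audio parameters.
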(event EventBeginPlay
  (|CallParentFunction)
  (Utilities|FlowControl|Sequence
    (:then_0
      (Networking|SwitchHasAuthority
        (:Authority
          (Initialization|InitializeDefaultState)
          ;[0.15 0.1361 0.01033 0.5] Bind Event to Trigger Intro Animation
          (EventDispatchers|BindEventtoOnStateTransitionedEvent (Variables|Default|GetBP_MajorSpireGameplayComponent) (delegate Custom|OnStateTransitionedEvent))
          (if (Variables|ForCreative|GetbStartLogicInBeginPlay)
            (StateFlow|StartStateLogic)))))
    (:then_1
      (bind _asspire_boss_anim_bp (Utilities|Casting|CastToSpireBoss_AnimBP (Components|SkeletalMesh|GetAnimInstance (Variables|Default|GetSkeletalMeshComponentBudgeted))))
      (Variables|Default|SetSpireBoss_Anim%20BP _asspire_boss_anim_bp))
    (:then_2
      (Utilities|IsValid (Variables|Default|GetBP_SpiresAnalyticsComponent)
        (:Is%20Valid
          (Utilities|IsValid (Variables|InternalData|GetEncounterManager)
            (:Is%20Valid
              (Default|BindEventtoOnGuardsSpawned (Variables|InternalData|GetEncounterManager) (delegate Custom|OnGuardsSpawnedEvent))
              (Default|BindEventtoOnGuardEliminated (Variables|InternalData|GetEncounterManager) (delegate Custom|OnGuardEliminatedEvent))
              (Default|BindEventtoOnActorSpawnComplete (Variables|InternalData|GetEncounterManager) (delegate Custom|OnEncManagerSpawnComplete)))))))
    (:then_3
      (Networking|SwitchHasAuthority
        (:Remote
          (CallFunction|ShouldbeShielded))))))

(event Custom|OnStateTransitionedEvent
  (params (PreviousState :type Gameplay%20Tag%20Structure%20%28by%20ref%29) (NewState :type Gameplay%20Tag%20Structure%20%28by%20ref%29))
  ;[0.15 0.1361 0.01033 0.5] Bind Event to Trigger Intro Animation
  (if (or (== NewState "(TagName=\"Spire.Major.State.Active.Idle\")") (== NewState "(TagName=\"Spire.Major.State.Active.Combat\")"))
    (CallFunction|PlayLookatAudio)
    (if (not (Variables|ForCreative|GetbFinishedIntroSequence))
      (Variables|ForCreative|SetbFinishedIntroSequence true)
      (Default|CallOnPlungeFinished))))

(event Custom|OnEncManagerSpawnComplete
  (GameplayEffects|RemoveActiveEffectswithGrantedTags (Ability|GetAbilitySystemComponent 0) "(GameplayTags=((TagName=\"Ability.SB.Summon.Summoning\")))"))

(event Custom|OnGuardsSpawnedEvent
  (params (NumSpawned :type Integer))
  (Class|SpiresAnalyticsComponent|OnGuardsSpawned (Variables|Default|GetBP_SpiresAnalyticsComponent) NumSpawned))

(event Custom|OnGuardEliminatedEvent
  (Class|SpiresAnalyticsComponent|OnGuardEliminated (Variables|Default|GetBP_SpiresAnalyticsComponent)))

(event Custom|SetScaledHealth
  ;[0 0.6719 0.1154 0.5] Set Scaled Health Value
  (bind (_gamestate _playlist) (Class|FortAsyncActionCurrentPlaylistReady|PlaylistDataReadyAsync)
    (:OnReady
      (Variables|InternalData|SetbHealthInitialized true)
      (Variables|InternalData|SetMaxSpireTeamSize (Variables|/Script/FortniteGame.FortPlaylist|GetMaxTeamSize _playlist))
      (Health|SetMaxHealthDontCreateASC self (* (ScalableFloat|GetValueatLevel (Variables|InternalData|GetSF_DefaultHealthScalar) (Math|Conversions|ToFloat(Integer) (Variables|/Script/FortniteGame.FortPlaylist|GetMaxTeamSize _playlist))) 1e+02)))))

(event Custom|PerformBeamSweep
  ;[0.2633 0.02256 0.9387] Spire Beam Sweep Logic
  (CallFunction|BeamSweepUpdateOverTime)
  (Networking|ForceNetUpdate))

(event EventOnDamageServer
  (params (Damage :type Float%20%28single-precision%29) (DamageTags :type Gameplay%20Tag%20Container%20Structure%20%28by%20ref%29) (Momentum :type Vector :default "0, 0, 0") (HitInfo :type Hit%20Result%20Structure%20%28by%20ref%29) (InstigatedBy :type Controller%20Object%20Reference) (DamageCauser :type Actor%20Object%20Reference) (EffectContext :type Gameplay%20Effect%20Context%20Handle%20Structure))
  ;[1 0.5314 0.0828] Health Threshold Events
  (|CallParentFunction Damage DamageTags Momentum HitInfo InstigatedBy DamageCauser EffectContext)
  (bind _bcrossedthreshold (CallFunction|CheckforCrossedHealthThreshold))
  (if _bcrossedthreshold
    (Default|CallOnSpireDamageThresholdReached DamageTags HitInfo InstigatedBy DamageCauser)
    (Abilities|TryActivateAbilitybyClass (Ability|GetAbilitySystemComponent 0) (Variables|Default|GetFullBodyHitReactAbility) true)))

(event Custom|BeamSweepUpdateOverTime
  (Variables|BeamData|SetTransformBeamEndPos (Transformation|GetActorTransform (Variables|BeamData|GetBeamTargetPawn)))
  ;[0.2633 0.02256 0.9387] Spire Beam Sweep Logic
  (Components|Timeline|SetPlayRate (Variables|Components|GetBeamSweepTimeline) (Math|Float|SafeDivide 1.0 (ScalableFloat|GetValueatLevelwithContext (Variables|BeamData|GetSF_BeamFireTime))))
  (if (Math|Random|RandomBool)
    (Variables|Default|SetSFBeamYawStart (ScalableFloat|GetValueatLevelwithContext (Variables|BeamData|GetSF_BeamYawStart)))
    (Variables|Default|SetSFBeamYawEnd (ScalableFloat|GetValueatLevelwithContext (Variables|BeamData|GetSF_BeamYawEnd)))
    (else
      (Variables|Default|SetSFBeamYawStart (* (ScalableFloat|GetValueatLevelwithContext (Variables|BeamData|GetSF_BeamYawStart)) -1.0))
      (Variables|Default|SetSFBeamYawEnd (* (ScalableFloat|GetValueatLevelwithContext (Variables|BeamData|GetSF_BeamYawEnd)) -1.0))))
  (bind (_direction _newtrack_0) (Timeline|BeamSweepTimeline PlayFromStart)
    (:Update
      (bind _output_get (Variables|ForCreative|SetBeamSweepCurrentYaw (Math|Float|Lerp (Variables|Default|GetSFBeamYawStart) (Variables|Default|GetSFBeamYawEnd) _newtrack_0)))
      (CallFunction|UpdateBeamSweepTransform _output_get))
    (:Finished
      (Variables|ForCreative|SetBeamSweepCurrentYaw 0.0))))

(event Custom|UpdateMaterialHealth
  (params (Health :type Float%20%28double-precision%29))
  (bind _returnvalue -2.0)
  (bind _returnvalue_2 (* _returnvalue Health))
  (bind _returnvalue_3 (+ _returnvalue_2 2.0))
  (bind _returnvalue_4 (Math|Float|Truncate _returnvalue_3))
  (bind _ns_spire_damagestates (Variables|Default|GetNS_Spire_DamageStates))
  ;[1 0.5314 0.0828] Health Threshold Events
  (Rendering|Material|SetCustomPrimitiveDataFloat (CallFunction|GetSpireMesh) 0 Health) ; set health to 33%
  (Niagara|SetNiagaraVariable(Int32) (Variables|Default|GetNS_Spire_DamageStates) "SpireState" _returnvalue_4) ; 0 - none\n1 - 66 % State\n2 - 33 % State
  (if (> _returnvalue_4 0)
    (Components|Activation|Activate _ns_spire_damagestates true)
    (else
      (Components|Activation|Deactivate _ns_spire_damagestates)))
  (Niagara|SetNiagaraVariableByString(Bool) (Variables|Default|GetNS_Spire_DamageStates) "Death" (== Health 0))
  (Networking|CanExecuteCosmeticEvents
    (:True
      (Audio|Parameter|SetIntegerParameter (Variables|Default|GetLookAtAudioLoop) "Health" _returnvalue_4 :!class /Script/AudioExtensions.AudioParameterControllerInterface))))

(event Custom|PlayRicochet
  (params (Hit%20Result :type Hit%20Result%20Structure) (Instigator :type Fort%20Pawn%20Object%20Reference))
  (bind (_bblockinghit _binitialoverlap _time _distance _location _impactpoint _normal) (Collision|BreakHitResult Hit%20Result))
  ;[0.9271 0.1015 0.09266 0.5] Play Ricochet Hit VFX
  (Development|PrintString "@@@ PlayRicochet [BGA_Large_Spire]: CALLED!" false (Variables|Debug|GetbDebugActor))
  (Networking|CanExecuteCosmeticEvents
    (:True
      (Utilities|FlowControl|Sequence
        (:then_0
          (if (GameplayTags|HasMatchingGameplayTag self (Variables|InternalData|GetShieldedTags))
            (Variables|InternalData|SetRicochetFXSystem "/SpireSharedAssets/VFX/Temp/NS_MinorSpire_Ricochet.NS_MinorSpire_Ricochet")
            (else
              (Variables|InternalData|SetRicochetFXSystem "/SpireSharedAssets/VFX/NS_SpireBoss_RicochetImpact.NS_SpireBoss_RicochetImpact"))))
        (:then_1
          (bind _returnvalue (Effects|Components|ParticleSystem|SpawnFXAtLocation (Variables|InternalData|GetRicochetFXSystem) _location (Math|Conversions|RotationFromXVector _normal)))
          (Effects|Components|ParticleSystem|SetVectorParameter _returnvalue "ImpactFXDirection" _normal)
          (if (Pawn|IsLocallyViewed Instigator)
            (Audio|SpawnSoundatLocation "/SpireSharedAssets/Audio/Cues/SC_MinorSpire_Ricochet.SC_MinorSpire_Ricochet" _location)))))))

(event Custom|ShieldDelayRemove
  (Utilities|FlowControl|Delay 0.667)
  (Rendering|Material|SetOverlayMaterial (Variables|Default|GetSkeletalMeshComponentBudgeted) 0 false 0 :!class /Script/Engine.MeshComponent))

(event EventOnDeathServer
  (params (Damage :type Float%20%28single-precision%29) (DamageTags :type Gameplay%20Tag%20Container%20Structure%20%28by%20ref%29) (Momentum :type Vector :default "0, 0, 0") (HitInfo :type Hit%20Result%20Structure%20%28by%20ref%29) (InstigatedBy :type Controller%20Object%20Reference) (DamageCauser :type Actor%20Object%20Reference) (EffectContext :type Gameplay%20Effect%20Context%20Handle%20Structure))
  ;[0.03264 0.5 0.2973 0.5] Handle Spire Death VFX
  (Destruction|HandleonDeathServer InstigatedBy))

(event Custom|OnDeathShockwave
  (bind _encountermanager (Variables|InternalData|GetEncounterManager))
  ;[0.03264 0.5 0.2973 0.5] Handle Spire Death VFX
  (Networking|SwitchHasAuthority
    (:Authority
      (Class|AthenaFunctionLibrary|SpawnLootfromTableUsingItemEntry (CallFunction|GetLargeSpireLootTier) 1 -1 (+ (Transformation|GetActorLocation self) (* (Transformation|GetActorUpVector self) 1e+03)) (Transformation|GetActorForwardVector self) 0 true true false 0 "Other" "NPCElimination")
      (Utilities|IsValid _encountermanager
        (:Is%20Valid
          (Class|BPSpireEncounterManager|EventKillEncounterActors _encountermanager (ScalableFloat|GetValueasBool (Variables|InternalData|GetSF_UseEncounterKillOffsets))))))))

(event Custom|OnNearbyActorDetected
  (params (AllFoundActors :type Array%20of%20Actor%20Object%20References) (EnteringActors :type Array%20of%20Actor%20Object%20References) (ExitingActors :type Array%20of%20Actor%20Object%20References))
  (bind _targetupdatetimerhandle (Variables|InternalData|GetTargetUpdateTimerHandle))
  (bind _returnvalue (Utilities|Time|IsTimerActivebyHandle _targetupdatetimerhandle))
  (bind _encountermanager (Variables|InternalData|GetEncounterManager))
  (bind _encountermanager_2 (Variables|InternalData|GetEncounterManager))
  ;[0.9479 0.3295 0.05314 0.5] Monitor Player Pawn Sensors
  (for _array_element EnteringActors
    (bind _asfort_player_pawn (Utilities|Casting|CastToFortPlayerPawn _array_element))
    (Utilities|Array|Add (Variables|InternalData|GetSensedPlayerPawns) _asfort_player_pawn)
    (Utilities|IsValid _encountermanager_2
      (:Is%20Valid
        (Class|BPSpireEncounterManager|AddTarget _encountermanager_2 _asfort_player_pawn))))
  (for _array_element_2 ExitingActors
    (bind _asfort_player_pawn_2 (Utilities|Casting|CastToFortPlayerPawn _array_element_2))
    (Utilities|Array|RemoveItem (Variables|InternalData|GetSensedPlayerPawns) _asfort_player_pawn_2)
    (Utilities|IsValid _encountermanager
      (:Is%20Valid
        (Class|BPSpireEncounterManager|RemoveTarget _encountermanager 0))))
  (bind _output_get (Variables|InternalData|SetSensedPlayerPawnCount (Utilities|Array|Length (PawnTracking|GetSensedTargets))))
  (if (and (> _output_get 0) (not _returnvalue))
    (CallFunction|UpdateCurrentTarget)
    (elif (and _returnvalue (== _output_get 0))
      (Utilities|Time|ClearandInvalidateTimerbyHandle (Variables|InternalData|GetTargetUpdateTimerHandle)))))

(event Custom|ToggleVisibility
  ;[0.9821 0.1771 1] Update Spire Materials
  (Rendering|SetVisibility (CallFunction|GetSpireMesh) false true))

(event Custom|RelayDeathToAnimBP
  ;[0.9821 0.1771 1] Update Spire Materials
  (Class|SpireBossAnimBP|SetIsDead (Variables|Default|GetSpireBoss_Anim%20BP) true)
  (CallFunction|AssignDeathMaterials))

(event EventOnDamagePlayEffects
  (params (Damage :type Float%20%28single-precision%29) (DamageTags :type Gameplay%20Tag%20Container%20Structure%20%28by%20ref%29) (Momentum :type Vector :default "0, 0, 0") (HitInfo :type Hit%20Result%20Structure%20%28by%20ref%29) (InstigatedBy :type Fort%20Pawn%20Object%20Reference) (DamageCauser :type Actor%20Object%20Reference) (EffectContext :type Gameplay%20Effect%20Context%20Handle%20Structure))
  (Utilities|Casting|CastToBGA_WeakpointActor_Base DamageCauser
    (:then
      (CallFunction|RelayHitReactionstoAnimBP DamageCauser))
    (:CastFailed
      (CallFunction|PlayRicochet HitInfo InstigatedBy))))

(event Custom|UpdateChildWeakpointMaterials
  (params (Target%20Weakpoint%20Kind :type WeakpointKind%20Enum :default "NewEnumerator0") (Enable%20Weakpoint :type Boolean))
  ;[0.9821 0.1771 1] Update Spire Materials
  (switch Utilities|FlowControl|Switch|SwitchonWeakpointKind Target%20Weakpoint%20Kind
    (:NewEnumerator0
      (CallFunction|HandleFrontLeftAppendageActive Enable%20Weakpoint))
    (:NewEnumerator1
      (CallFunction|HandleFrontRightAppendageActive Enable%20Weakpoint))
    (:NewEnumerator2
      (CallFunction|HandleBackAppendageActive Enable%20Weakpoint))
    (:NewEnumerator3
      (CallFunction|HandleCoreActive Enable%20Weakpoint))))

(event Custom|ApplyChildWeakpointMaterials
  (bind _spiremesh (CallFunction|GetSpireMesh))
  (bind _bspirecoreactive (Variables|WeakpointData|GetbSpireCoreActive))
  (bind _returnvalue (Math|Float|SelectFloat 1.0 0.0 _bspirecoreactive))
  ;[0.9821 0.1771 1] Update Spire Materials
  (Rendering|Material|SetCustomPrimitiveDataFloat _spiremesh 1 (Math|Float|SelectFloat 1.0 0.0 (Variables|WeakpointData|GetbFrontRightAppendageActive))) ; set health to 33%
  (Rendering|Material|SetCustomPrimitiveDataFloat _spiremesh 2 (Math|Float|SelectFloat 1.0 0.0 (Variables|WeakpointData|GetbFrontLeftAppendageActive))) ; set health to 33%
  (Rendering|Material|SetCustomPrimitiveDataFloat _spiremesh 3 (Math|Float|SelectFloat 1.0 0.0 (Variables|WeakpointData|GetbBackAppendageActive))) ; set health to 33%
  (Rendering|Material|SetCustomPrimitiveDataFloat _spiremesh 4 _returnvalue) ; set health to 33%
  (Rendering|Material|SetCustomPrimitiveDataFloat _spiremesh 5 _returnvalue) ; set health to 33%
  (Rendering|Material|SetCustomPrimitiveDataFloat _spiremesh 6 (Math|Float|SelectFloat 1.0 0.0 (Math|Boolean|ORBoolean (Variables|WeakpointData|GetbFrontRightAppendageActive) (Variables|WeakpointData|GetbFrontLeftAppendageActive) (Variables|WeakpointData|GetbBackAppendageActive) (Variables|WeakpointData|GetbSpireCoreActive)))) ; set health to 33%
)

(event Custom|RequestPlunge
  ;[0.2812 0.05609 0.2383 0.5] Intro Plunge
  (Variables|Plunge|SetTargetPlungeLocation (Transformation|GetActorLocation))
  (CallFunction|StartPlunge))

(event Custom|StartPlunge
  ;[0.2812 0.05609 0.2383 0.5] Intro Plunge
  (Components|Timeline|SetPlayRate (Variables|BGA_Large_Spire|GetPlungeDown) (/ 1.0 (ScalableFloat|GetValueatLevel (Variables|Default|GetPlungeDuration))))
  (Variables|Plunge|SetStartingPlungeLocation (+ (* (Math|Vector|VectorUp) (ScalableFloat|GetValueatLevel (Variables|Plunge|GetPlungeHeight))) (Transformation|GetActorLocation)))
  (bind (_direction _alpha) (Timeline|PlungeDown PlayFromStart)
    (:Update
      (Transformation|SetActorLocation (Math|Vector|Lerp(Vector) (Variables|Plunge|GetStartingPlungeLocation) (Variables|Plunge|GetTargetPlungeLocation) _alpha) false true))
    (:Finished
      (GameplayCue|ExecuteGameplayCue%28Burst%29Local self (Variables|Plunge|GetGC_PlungeExplosion) (Ability|GameplayCue|MakeGameplayCueParameters :NormalizedMagnitude 0.0 :RawMagnitude 0.0 :Location (Variables|Plunge|GetTargetPlungeLocation) :Normal "0,0,0" :GameplayEffectLevel 1 :AbilityLevel 1 :bReplicateLocationWhenUsingMinimalRepProxy false))
      (Default|CallOnPlungeFinished))) ; Using a timeline per design request so they have more control \nover the curve until we have a root motion animation (if we get one).\n Will probably have to nativize if we end up using this for the final
)

(event Custom|PlayWeakpointFadeOut
  (params (FadeIn :type Boolean) (DisableOnFinished :type Boolean))
  (bind _spiremesh (CallFunction|GetSpireMesh))
  ;[0.9821 0.1771 1] Update Spire Materials
  (if FadeIn
    (bind (_direction _fadeout) (Timeline|TL_WeakpointFade ReverseFromEnd)
      (:Update
        (Rendering|Material|SetCustomPrimitiveDataFloat (CallFunction|GetSpireMesh) 6 _fadeout) ; set health to 33%
      )
      (:Finished
        (if DisableOnFinished
          (Rendering|Material|SetCustomPrimitiveDataFloat _spiremesh 1 0.0) ; set health to 33%
          (Rendering|Material|SetCustomPrimitiveDataFloat _spiremesh 2 0.0) ; set health to 33%
          (Rendering|Material|SetCustomPrimitiveDataFloat _spiremesh 3 0.0) ; set health to 33%
          (Rendering|Material|SetCustomPrimitiveDataFloat _spiremesh 4 0.0) ; set health to 33%
          (Rendering|Material|SetCustomPrimitiveDataFloat _spiremesh 5 0.0) ; set health to 33%
          (Rendering|Material|SetCustomPrimitiveDataFloat _spiremesh 6 0.0) ; set health to 33%
        )))
    (else
      (Timeline|TL_WeakpointFade PlayFromStart))))

(event Custom|RemoveChildWeakpointMaterials
  (bind _spiremesh (CallFunction|GetSpireMesh))
  ;[0.9821 0.1771 1] Update Spire Materials
  (CallFunction|HandleFrontLeftAppendageActive false)
  (CallFunction|HandleFrontRightAppendageActive false)
  (CallFunction|HandleBackAppendageActive false)
  (CallFunction|HandleCoreActive false)
  (Rendering|Material|SetCustomPrimitiveDataFloat _spiremesh 1 0.0) ; set health to 33%
  (Rendering|Material|SetCustomPrimitiveDataFloat _spiremesh 2 0.0) ; set health to 33%
  (Rendering|Material|SetCustomPrimitiveDataFloat _spiremesh 3 0.0) ; set health to 33%
  (Rendering|Material|SetCustomPrimitiveDataFloat _spiremesh 4 0.0) ; set health to 33%
  (Rendering|Material|SetCustomPrimitiveDataFloat _spiremesh 5 0.0) ; set health to 33%
  (Rendering|Material|SetCustomPrimitiveDataFloat _spiremesh 6 0.0) ; set health to 33%
)

(event Custom|PlayLookAtAudio
  (Networking|CanExecuteCosmeticEvents
    (:True
      (Audio|Components|Audio|FadeIn (Variables|Default|GetLookAtAudioLoop) 0.5 1.0 0.0 "Linear" :!class /Script/Engine.AudioComponent))))

(event Custom|HandleFrontLeftAppendageActive
  (params (bNewFrontLeftAppendageActive :type Boolean))
  ;[0.9821 0.1771 1] Update Spire Materials
  (bind _output_get (Variables|WeakpointData|SetbFrontLeftAppendageActive bNewFrontLeftAppendageActive))
  (WeakpointMaterials|ToggleWeakpointMaterialActive 2 _output_get))

(event Custom|Play%20Death%20Stinger
  (MusicEncounter|ExecuteMusicEncounterGameplayEventByTag (GameplayEventRouter|GetGlobalEventRouter) (Variables|InternalData|GetExplodingTag) self))

(event Custom|HandleFrontRightAppendageActive
  (params (bNewFrontRightAppendageActive :type Boolean))
  ;[0.9821 0.1771 1] Update Spire Materials
  (bind _output_get (Variables|WeakpointData|SetbFrontRightAppendageActive bNewFrontRightAppendageActive))
  (WeakpointMaterials|ToggleWeakpointMaterialActive 1 _output_get))

(event Custom|Activate%20Music
  ; Music activation
  (MusicEncounter|ExecuteMusicEncounterGameplayEventByName (GameplayEventRouter|GetGlobalEventRouter) "Activate" self))

(event Custom|HandleBackAppendageActive
  (params (bNewBackAppendageActive :type Boolean))
  ;[0.9821 0.1771 1] Update Spire Materials
  (bind _output_get (Variables|WeakpointData|SetbBackAppendageActive bNewBackAppendageActive))
  (WeakpointMaterials|ToggleWeakpointMaterialActive 3 _output_get))

(event Custom|Deactivate%20Music
  ; Music activation
  (MusicEncounter|ExecuteMusicEncounterGameplayEventByName (GameplayEventRouter|GetGlobalEventRouter) "Deactivate" self))

(event Custom|HandleCoreActive
  (params (bNewCoreActive :type Boolean))
  ;[0.9821 0.1771 1] Update Spire Materials
  (bind _output_get (Variables|WeakpointData|SetbSpireCoreActive bNewCoreActive))
  (WeakpointMaterials|ToggleWeakpointMaterialActive 4 _output_get)
  (WeakpointMaterials|ToggleWeakpointMaterialActive 5 _output_get))

(event Custom|SkipMajorSpireIntroWait
  ;[1 1 1] Super Proto. Want to make some more flexible way to send debug events to the state tree in the future
  (Gameplay|StateTree|SendStateTreeEvent (Variables|Default|GetStateTree) (Utilities|Struct|MakeStateTreeEvent "(TagName=\"Spire.Major.State.Intro.SkipActivationWait\")") :!class /Script/GameplayStateTreeModule.StateTreeComponent))

(event Custom|TestSetWeakPoint
  (params (Type :type Integer) (Enable :type Boolean))
  (CallFunction|UpdateChildWeakpointActorbyKind (Utilities|Select :Index Type :Option%200 "NewEnumerator0" :Option%201 "NewEnumerator1" :Option%202 "NewEnumerator2" :Option%203 "NewEnumerator3") Enable))

(event Custom|TestSetShield
  (params (bEnable :type Boolean))
  (Variables|InternalData|SetbShouldbeShielded bEnable))

(event Custom|TestDestroySpire
  (bind _weakpoint04 (Variables|Default|GetWeakpoint04))
  (bind _childactor (Variables|/Script/Engine.ChildActorComponent|GetChildActor _weakpoint04))
  (bind _asbga_weakpoint_actor_core_large_spire (Utilities|Casting|CastToBGA_WeakpointActor_Core_Large_Spire _childactor))
  (Class|BGAWeakpointActorBase|EnableWeakpoint _asbga_weakpoint_actor_core_large_spire)
  (Game|Damage|ApplyDamage self 999999999.0 0 _childactor))

(fn AssignDeathMaterials
  (bind _spiremesh (CallFunction|GetSpireMesh))
  (Rendering|Material|SetMaterial _spiremesh 0 "/SpireSharedAssets/Assets/Materials/SpireDeathMaterials/MI_SpireTop_Death.MI_SpireTop_Death")
  (Rendering|Material|SetMaterial _spiremesh 1 "/SpireSharedAssets/Assets/Materials/SpireDeathMaterials/MI_SpireFloat_Death.MI_SpireFloat_Death")
  (Rendering|Material|SetMaterial _spiremesh 2 "/SpireSharedAssets/Assets/Materials/SpireDeathMaterials/MI_SpireBottom_Death.MI_SpireBottom_Death")
  (Rendering|Material|SetMaterial _spiremesh 3 "/SpireSharedAssets/Assets/Materials/SpireDeathMaterials/MI_SpireCore_Death.MI_SpireCore_Death"))

(fn BlueprintModifyIncomingDamage
  (params (Damage :type Float%20%28single-precision%29) (DamageTags :type Gameplay%20Tag%20Container%20Structure%20%28by%20ref%29) (EffectContext :type Gameplay%20Effect%20Context%20Handle%20Structure%20%28by%20ref%29) (EventInstigator :type Controller%20Object%20Reference) (DamageCauser :type Actor%20Object%20Reference))
  (|CallParentFunction Damage DamageTags EffectContext EventInstigator DamageCauser)
  (Development|PrintString "@@@ BlueprintModifyIncomingDamage [BGA_Large_Spire]: CALLED!" false (Variables|Debug|GetbDebugActor))
  (if (Variables|InternalData|GetbHealthInitialized)
    (if (GameplayTags|HasMatchingGameplayTag self (Variables|InternalData|GetShieldedTags))
      (return 0.0)
      (else
        (bind _bisweakpoint (Class|BPActorWeakpointsUtilityComponent|IsOwnedWeakpointActor (Variables|Default|GetBP_ActorWeakpoints_UtilityComponent) DamageCauser true))
        (if _bisweakpoint
          (Development|PrintString (Utilities|String|Append "@@@@@@ BlueprintModifyIncomingDamage [BGA_Large_Spire]: WEAKPOINT DAMAGE = " (Utilities|String|ToString(Float) Damage) "!") false (Variables|Debug|GetbDebugActor))
          (return Damage)
          (else
            (return (/ Damage (Math|Float|SelectFloat (ScalableFloat|GetValueatLevel (Variables|InternalData|GetSF_SuperDamageDivider)) (ScalableFloat|GetValueatLevel (Variables|InternalData|GetSF_BaseDamageDivider)) (GameplayTags|HasAnyTags DamageTags (Variables|InternalData|GetTagContainer_SupermanDamage) true))))))))
    (else
      (return 0.0))))

(fn CheckForCrossedHealthThreshold
  (bind _self self)
  (bind _returnvalue (Health|GetHealthPercent _self))
  (bind _temphealththreshold (LocalVariables|GetTempHealthThreshold))
  (bind _bdebugactor (Variables|Debug|GetbDebugActor))
  (Utilities|FlowControl|Sequence
    (:then_0
      ;[1 1 1] Check for change in health threshold
      (bind (_array_element _array_index) (Utilities|Array|ForEachLoopwithBreak (Variables|Default|GetHealthThresholdPercentages) :!label _label)
        (:LoopBody
          (if (> (Health|GetHealthPercent self) _array_element)
            (goto _label Break)
            (else
              (LocalVariables|SetTempHealthThreshold (+ _array_index 1)))))
        (:Completed
          (bind _output_get (LocalVariables|SetNewHealthThresholdReached (> _temphealththreshold (Variables|Default|GetCurrentHealthThreshold))))
          (if _output_get
            (Variables|Default|SetCurrentHealthThreshold _temphealththreshold)))))
    (:then_1
      ;[1 1 1] (Previous Behavior) Update material/vfx (keeping the 66% and 33% values regardless of actual # of thresholds since we only have 3 visual states
      (if (or (LocalVariables|GetNewHealthThresholdReached) (Utilities|Array|IsEmpty (Variables|Default|GetHealthThresholdPercentages)))
        (if (<= _returnvalue 0.66)
          (if (<= _returnvalue 0.33)
            (if (not (Variables|InternalData|GetbSecondHealthThresholdReached))
              (Variables|InternalData|SetbSecondHealthThresholdReached true)
              (CallFunction|UpdateMaterialHealth 0.0)
              (Development|PrintString "@@@ OnDamageServer [BGA_Large_Spire]: 33% - CALLED!" false _bdebugactor))
            (elif (not (Variables|InternalData|GetbFirstHealthThresholdReached))
              (Variables|InternalData|SetbFirstHealthThresholdReached true)
              (CallFunction|UpdateMaterialHealth 0.5)
              (Development|PrintString "@@@ OnDamageServer [BGA_Large_Spire]: 66% - CALLED!" false _bdebugactor))))))
    (:then_2
      (return (LocalVariables|GetNewHealthThresholdReached)))))

(fn GetAppendageDataByKind
  (params (TargetAppendageKind :type SpireAppendageKind%20Enum :default "NewEnumerator0"))
  (bind _spireappendagestruct (Variables|InternalData|GetSpireAppendageStruct))
  (LocalVariables|SetTestAppendageStruct (Utilities|Array|Get(acopy) _spireappendagestruct))
  (bind _array_element (Utilities|Array|ForEachLoopwithBreak _spireappendagestruct :!label _label)
    (:LoopBody
      (if (Utilities|Enum|Equal(Enum) (Utilities|Struct|BreakSpireAppendageStruct _array_element) (Variables|GetTargetAppendageKind))
        (LocalVariables|SetTestAppendageStruct _array_element)
        (goto _label Break)))
    (:Completed
      (return (LocalVariables|GetTestAppendageStruct)))))

(fn GetBeamStopTransform
  (return (Variables|BeamData|GetTransformBeamStopPos)))

(fn GetChildWeakpointActorByKind
  (params (TargetWeakpointKind :type WeakpointKind%20Enum :default "NewEnumerator0"))
  (switch Utilities|FlowControl|Switch|SwitchonWeakpointKind TargetWeakpointKind
    (:NewEnumerator0
      (bind _asbga_weakpoint_actor_base (Utilities|Casting|CastToBGA_WeakpointActor_Base (Variables|/Script/Engine.ChildActorComponent|GetChildActor (Variables|Default|GetWeakpoint02)))
        (:then
          (return _asbga_weakpoint_actor_base))
        (:CastFailed
          (return 0))))
    (:NewEnumerator1
      (bind _asbga_weakpoint_actor_base_2 (Utilities|Casting|CastToBGA_WeakpointActor_Base (Variables|/Script/Engine.ChildActorComponent|GetChildActor (Variables|Default|GetWeakpoint03)))
        (:then
          (return _asbga_weakpoint_actor_base_2))
        (:CastFailed
          (return 0))))
    (:NewEnumerator2
      (bind _asbga_weakpoint_actor_base_3 (Utilities|Casting|CastToBGA_WeakpointActor_Base (Variables|/Script/Engine.ChildActorComponent|GetChildActor (Variables|Default|GetWeakpoint01)))
        (:then
          (return _asbga_weakpoint_actor_base_3))
        (:CastFailed
          (return 0))))
    (:NewEnumerator3
      (bind _asbga_weakpoint_actor_base_4 (Utilities|Casting|CastToBGA_WeakpointActor_Base (Variables|/Script/Engine.ChildActorComponent|GetChildActor (Variables|Default|GetWeakpoint04)))
        (:then
          (return _asbga_weakpoint_actor_base_4))
        (:CastFailed
          (return 0))))))

(fn GetClosestAppendageKind
  (params (TestLocation :type Vector))
  (bind (_targetstructdata_spireappendagekind_2_c321b3314c9291adb879629b5339498e _targetstructdata_bonename_5_38cd39514d738e05454978b3e574f8a0) (Utilities|Struct|BreakSpireAppendageStruct (CallFunction|GetAppendageDatabyKind "NewEnumerator0")))
  (bind _skeletalmeshcomponentbudgeted (Variables|Default|GetSkeletalMeshComponentBudgeted))
  (bind _returnvalue_location (Math|Transform|BreakTransform (Transformation|GetBoneTransform _skeletalmeshcomponentbudgeted _targetstructdata_bonename_5_38cd39514d738e05454978b3e574f8a0 "RTS_World")))
  (bind (_targetstructdata_spireappendagekind_2_c321b3314c9291adb879629b5339498e_2 _targetstructdata_bonename_5_38cd39514d738e05454978b3e574f8a0_2) (Utilities|Struct|BreakSpireAppendageStruct (CallFunction|GetAppendageDatabyKind "NewEnumerator1")))
  (bind _skeletalmeshcomponentbudgeted_2 (Variables|Default|GetSkeletalMeshComponentBudgeted))
  (bind _returnvalue_location_2 (Math|Transform|BreakTransform (Transformation|GetBoneTransform _skeletalmeshcomponentbudgeted_2 _targetstructdata_bonename_5_38cd39514d738e05454978b3e574f8a0_2 "RTS_World")))
  (bind _testlocation (Variables|GetTestLocation))
  (bind _returnvalue (Math|Vector|Distance(Vector) _testlocation _returnvalue_location_2))
  (bind (_targetstructdata_spireappendagekind_2_c321b3314c9291adb879629b5339498e_3 _targetstructdata_bonename_5_38cd39514d738e05454978b3e574f8a0_3) (Utilities|Struct|BreakSpireAppendageStruct (CallFunction|GetAppendageDatabyKind "NewEnumerator2")))
  (bind _skeletalmeshcomponentbudgeted_3 (Variables|Default|GetSkeletalMeshComponentBudgeted))
  (bind _returnvalue_location_3 (Math|Transform|BreakTransform (Transformation|GetBoneTransform _skeletalmeshcomponentbudgeted_3 _targetstructdata_bonename_5_38cd39514d738e05454978b3e574f8a0_3 "RTS_World")))
  (bind _testlocation_2 (Variables|GetTestLocation))
  (bind _returnvalue_2 (Math|Vector|Distance(Vector) _testlocation_2 _returnvalue_location_3))
  (bind _testlocation_3 (Variables|GetTestLocation))
  (bind _returnvalue_3 (Math|Vector|Distance(Vector) _testlocation_3 _returnvalue_location))
  (Development|PrintString "@@@ GetClosestAppendageKind [BGA_Large_Spire]: CALLED!" false (Variables|Debug|GetbDebugActor))
  (Variables|Default|SetCurrentClosestAppendageKind "NewEnumerator0")
  (if (< _returnvalue_3 _returnvalue)
    (Variables|Default|SetCurrentClosestAppendageKind "NewEnumerator0")
    (bind _output_get (LocalVariables|SetTestDistance _returnvalue_3))
    (if (not (< _output_get _returnvalue_2))
      (Variables|Default|SetCurrentClosestAppendageKind "NewEnumerator2" :!label _label))
    (else
      (Variables|Default|SetCurrentClosestAppendageKind "NewEnumerator1")
      (bind _output_get_2 (LocalVariables|SetTestDistance _returnvalue))
      (if (not (< _output_get_2 _returnvalue_2))
        (goto _label))))
  (Development|PrintString (Utilities|String|Append "@@@@@@ GetClosestAppendageKind [BGA_Large_Spire]: " (Utilities|String|EnumtoString (Variables|Default|GetCurrentClosestAppendageKind)) " is closest!") false (Variables|Debug|GetbDebugActor))
  (return (Variables|Default|GetCurrentClosestAppendageKind)))

(fn GetData
  (bind (_returnvalue_location _returnvalue_rotation) (Math|Transform|BreakTransform (Transformation|GetBoneTransform (Variables|Default|GetSkeletalMeshComponentBudgeted) "shard_center" "RTS_World")))
  (bind (_currentbeamtransform_location _currentbeamtransform_rotation) (Math|Transform|BreakTransform (Variables|GetCurrentBeamTransform_Location)))
  (bind (_currentbeamtransform_rotation_roll _currentbeamtransform_rotation_pitch _currentbeamtransform_rotation_yaw) (Math|Rotator|BreakRotator _currentbeamtransform_rotation))
  (bind (_windupbeamtransform_location _windupbeamtransform_rotation) (Math|Transform|BreakTransform (Variables|GetWindupBeamTransform_Location)))
  (bind (_windupbeamtransform_rotation_roll _windupbeamtransform_rotation_pitch _windupbeamtransform_rotation_yaw) (Math|Rotator|BreakRotator _windupbeamtransform_rotation))
  (return (Utilities|Enum|Equal(Enum) (Variables|Default|GetCurrentClosestAppendageKind) "NewEnumerator0") (Utilities|Enum|Equal(Enum) (Variables|Default|GetCurrentClosestAppendageKind) "NewEnumerator1") (Utilities|Enum|Equal(Enum) (Variables|Default|GetCurrentClosestAppendageKind) "NewEnumerator2") _windupbeamtransform_rotation_yaw _currentbeamtransform_rotation_yaw _currentbeamtransform_rotation_pitch _returnvalue_rotation))

(fn GetFoundBlockingHit
  (return (Variables|BeamData|GetbFoundBlockingHit)))

(fn GetFoundWaterBlockingHit
  (return (Variables|BeamData|GetbFoundWaterBlockingHit)))

(fn GetLargeSpireLootTier
  (bind _returnvalue (Fortnite|Athena|GameState|GetGameStateAthena))
  (bind _returnvalue_2 (Actor|GetComponentbyClass _returnvalue "/SpireBoss/GameplayActors/Components/BP_GameStateComponent_SpireBoss.BP_GameStateComponent_SpireBoss_C"))
  (bind (_spirelootgroupstruct_sololootgroup_2_0efac2cc4da471bcfd5222889d63e662 _spirelootgroupstruct_duolootgroup_4_28d752d545e039602f638c8fe767a9a8 _spirelootgroupstruct_triolootgroup_6_e655b8744f75d84c01b8d1aaf0e02f58 _spirelootgroupstruct_squadlootgroup_8_a55d0f7647d8c26836167792a05b658e) (Utilities|Struct|BreakSpireLootGroupStruct (Class|BPGameStateComponentSpireBoss|GetNextSpireLootGroup _returnvalue_2)))
  (switch int (Variables|InternalData|GetMaxSpireTeamSize)
    (:1
      (return _spirelootgroupstruct_sololootgroup_2_0efac2cc4da471bcfd5222889d63e662))
    (:2
      (return _spirelootgroupstruct_duolootgroup_4_28d752d545e039602f638c8fe767a9a8))
    (:3
      (return _spirelootgroupstruct_triolootgroup_6_e655b8744f75d84c01b8d1aaf0e02f58))
    (:4
      (return _spirelootgroupstruct_squadlootgroup_8_a55d0f7647d8c26836167792a05b658e))))

(fn GetLineTraceIgnoredActors
  (return (Utilities|Array|MakeArray (Variables|/Script/Engine.ChildActorComponent|GetChildActor (Variables|Default|GetWeakpoint01)) (Variables|/Script/Engine.ChildActorComponent|GetChildActor (Variables|Default|GetWeakpoint02)) (Variables|/Script/Engine.ChildActorComponent|GetChildActor (Variables|Default|GetWeakpoint03)) (Variables|/Script/Engine.ChildActorComponent|GetChildActor (Variables|Default|GetWeakpoint04)) self)))

(fn GetSensedTargets
  (return (Variables|InternalData|GetSensedPlayerPawns)))

(fn GetSkeletalMeshComponentBudgeted
  (return (Variables|Default|GetSkeletalMeshComponentBudgeted)))

(fn GetSpireMesh
  (return (Variables|Default|GetSkeletalMeshComponentBudgeted)))

(fn GetStopHitActor
  (return (Variables|BeamData|GetStopHitActor)))

(fn GetStopHitNormal
  (return (Variables|BeamData|GetStopHitNormal)))

(fn GetWeakpointKindByActor
  (params (Actor :type Actor%20Object%20Reference))
  (bind _asbga_weakpoint_actor_large_spire (Utilities|Casting|CastToBGA_WeakpointActor_Large_Spire Actor)
    (:then
      (if (== _asbga_weakpoint_actor_large_spire (Variables|/Script/Engine.ChildActorComponent|GetChildActor (Variables|Default|GetWeakpoint01)))
        (return "NewEnumerator2")
        (elif (== _asbga_weakpoint_actor_large_spire (Variables|/Script/Engine.ChildActorComponent|GetChildActor (Variables|Default|GetWeakpoint02)))
          (return "NewEnumerator0")
          (elif (== _asbga_weakpoint_actor_large_spire (Variables|/Script/Engine.ChildActorComponent|GetChildActor (Variables|Default|GetWeakpoint03)))
            (return "NewEnumerator1")))))
    (:CastFailed
      (Utilities|Casting|CastToBGA_WeakpointActor_Core_Large_Spire Actor)
      (return "NewEnumerator3"))))

(fn GetWeakspotActors
  (return (Variables|/ActorWeakpointsCore/Components/BP_ActorWeakpoints_UtilityComponent.BP_ActorWeakpoints_UtilityComponent_C|GetWeakpointActorArray (Variables|Default|GetBP_ActorWeakpoints_UtilityComponent))))

(fn GetWeakspotComponents
  (return 0))

(fn HandleOnDeathServer
  (params (InstigatorController :type Controller%20Object%20Reference))
  (bind _encountermanager (Variables|InternalData|GetEncounterManager))
  (Development|PrintString "@@@ OnDeathServer [BGA_Large_Spire]: CALLED!" false (Variables|Debug|GetbDebugActor))
  (Utilities|FlowControl|Sequence
    (:then_0
      (Default|CallOnSpireDefeated))
    (:then_1
      (CallFunction|RelayDeathtoAnimBP)
      (Utilities|Time|ClearandInvalidateTimerbyHandle (Variables|InternalData|GetTargetUpdateTimerHandle))
      (CallFunction|RemoveChildWeakpointMaterials)
      (Class|BPFLAICommonFunctions|AddUniqueGameplayTag self (Variables|InternalData|GetExplodingTag) true) ; FNE: revent stacking tag multiple times if spamming Destroy functionality
      (CallFunction|PlayDeathStinger))
    (:then_2
      (Destruction|HandleWeakpointsonSpireDestroyed))
    (:then_3
      (Ability|SendGameplayEventtoActor InstigatorController "(TagName=\"Event.SB.Eliminated\")"))
    (:then_4
      (Utilities|IsValid _encountermanager
        (:Is%20Valid
          (Default|UnbindallEventsfromOnGuardsSpawned _encountermanager)
          (Default|UnbindallEventsfromOnGuardEliminated _encountermanager))))))

(fn HandleWeakpointsOnSpireDestroyed
  (bind _asbga_weakpoint_actor_base (Utilities|Casting|CastToBGA_WeakpointActor_Base (Variables|/Script/Engine.ChildActorComponent|GetChildActor (Variables|Default|GetWeakpoint01))))
  (Class|BGAWeakpointActorBase|DestroyWeakpoint _asbga_weakpoint_actor_base)
  (bind _asbga_weakpoint_actor_base_2 (Utilities|Casting|CastToBGA_WeakpointActor_Base (Variables|/Script/Engine.ChildActorComponent|GetChildActor (Variables|Default|GetWeakpoint02))))
  (Class|BGAWeakpointActorBase|DestroyWeakpoint _asbga_weakpoint_actor_base_2)
  (bind _asbga_weakpoint_actor_base_3 (Utilities|Casting|CastToBGA_WeakpointActor_Base (Variables|/Script/Engine.ChildActorComponent|GetChildActor (Variables|Default|GetWeakpoint03))))
  (Class|BGAWeakpointActorBase|DestroyWeakpoint _asbga_weakpoint_actor_base_3))

(fn HasValidAbilityTarget
  (bind _targets (PawnTracking|GetSensedTargets))
  (bind _bfoundvalidtarget (LocalVariables|GetbFoundValidTarget))
  (if (> (Utilities|Array|Length _targets) 0)
    (LocalVariables|SetbFoundValidTarget false)
    (bind _array_element (Utilities|Array|ForEachLoopwithBreak _targets :!label _label)
      (:LoopBody
        (LocalVariables|SetCurrentTargetIter _array_element)
        (if (not (GameplayTags|HasAnyMatchingGameplayTags _array_element (GameplayTags|MakeLiteralGameplayTagContainer "(GameplayTags=((TagName=\"Gameplay.Status.DBNO\"),(TagName=\"Athena.Gameplay.Status.Cloaked\")))")))
          (LocalVariables|SetbFoundValidTarget true)
          (goto _label Break)))
      (:Completed
        (if _bfoundvalidtarget
          (return _bfoundvalidtarget (LocalVariables|GetCurrentTargetIter))
          (else
            (return _bfoundvalidtarget 0)))))
    (else
      (return false 0))))

(fn InitializeDefaultState
  (Utilities|FlowControl|Sequence
    (:then_0
      (Networking|SwitchHasAuthority
        (:Authority
          (Initialization|InitSpireVariables)
          (GameplayEffects|ApplyGameplayEffectToTarget :self (Ability|GetAbilitySystemComponent 0) :GameplayEffectClass (Variables|InternalData|GetThresholdCrossedSummon) :Target (Ability|GetAbilitySystemComponent 0) :Level 0.0)
          (CallFunction|SetScaledHealth))))
    (:then_1
      (Initialization|SetUpDefaultMaterials)
      (Networking|SwitchHasAuthority
        (:Authority
          (if (not (ScalableFloat|GetValueasBool (Variables|InternalData|GetSF_DebugDisableShieldVisuals)))
            (Variables|InternalData|SetbShouldbeShielded true)))))))

(fn InitSpireVariables
  (Networking|SwitchHasAuthority
    (:Authority
      (Variables|InternalData|SetbHealthInitialized false)
      (CallFunction|HandleFrontLeftAppendageActive false)
      (CallFunction|HandleFrontRightAppendageActive false)
      (CallFunction|HandleBackAppendageActive false)
      (CallFunction|HandleCoreActive false)
      (Variables|InternalData|SetbFirstHealthThresholdReached false)
      (Variables|InternalData|SetbSecondHealthThresholdReached false)
      (Variables|ForCreative|SetbFinishedIntroSequence false)
      (Variables|Default|SetCurrentHealthThreshold 0))) ; server-specific variables, some of which we want to rely on the OnRep of
)

(fn OnRep_bShouldbeShielded
  (CallFunction|ShouldbeShielded))

(fn OnRep_SpireRotation
  (Transformation|SetActorRotation 0 (0.0 0.0 0.0) false))

(fn OnRep_TargetPlungeLocation
  (CallFunction|StartPlunge))

(fn RelayHitReactionsToAnimBP
  (params (Actor :type Actor%20Object%20Reference))
  (bind _spireboss_anim_bp (Variables|Default|GetSpireBoss_Anim%20BP))
  (bind _weakpointkind (CallFunction|GetWeakpointKindbyActor Actor))
  (switch Utilities|FlowControl|Switch|SwitchonWeakpointKind _weakpointkind
    (:NewEnumerator0
      (Class|SpireBossAnimBP|HitReactShieldLeft _spireboss_anim_bp 1.0))
    (:NewEnumerator1
      (Class|SpireBossAnimBP|HitReactShieldRight _spireboss_anim_bp 1.0))
    (:NewEnumerator2
      (Class|SpireBossAnimBP|HitReactShieldRear _spireboss_anim_bp 1.0))
    (:NewEnumerator3
      (Class|SpireBossAnimBP|HitReactMain _spireboss_anim_bp 1.0))))

(fn SetCurrentTargetPlayer
  (params (NewTarget :type Fort%20Player%20Pawn%20Object%20Reference))
  (Variables|InternalData|SetCurrentTargetPlayerPawn NewTarget))

(fn SetUpDefaultMaterials
  (Rendering|Material|SetCustomPrimitiveDataFloat (CallFunction|GetSpireMesh) 0 1.0) ; initalize the heath value on the skel mesh
  (CallFunction|ApplyChildWeakpointMaterials))

(fn ShouldBeShielded
  (bind _skeletalmeshcomponentbudgeted (Variables|Default|GetSkeletalMeshComponentBudgeted))
  (if (ScalableFloat|GetValueasBool (Variables|InternalData|GetSF_DebugDisableShieldVisuals))
    (CallFunction|ApplyChildWeakpointMaterials)
    (elif (Variables|InternalData|GetbShouldbeShielded)
      (Rendering|Material|SetOverlayMaterial _skeletalmeshcomponentbudgeted 0 false 0 :!class /Script/Engine.MeshComponent)
      (Rendering|Material|SetOverlayMaterial _skeletalmeshcomponentbudgeted "/SpireSharedAssets/Assets/Materials/M_MinorSpire_Shield.M_MinorSpire_Shield" false 0 :!class /Script/Engine.MeshComponent)
      (bind _returnvalue (Class|FortCosmeticMaterialBlueprintLibrary|GetorCreateDynamicOverlayMaterial _skeletalmeshcomponentbudgeted))
      (Rendering|Material|SetScalarParameterValue _returnvalue "StartTime" (Utilities|Time|GetGameTimeinSeconds) :!class /Script/Engine.MaterialInstanceDynamic)
      (Class|BPFLAICommonFunctions|AddUniqueGameplayTag self (Variables|InternalData|GetShieldedTags)) ; FNE: Prevent stacking tag multiple times if spamming Reset functionality
      (CallFunction|PlayWeakpointFadeOut false true)
      (else
        (Utilities|FlowControl|Sequence
          (:then_0
            (Class|BPFLAICommonFunctions|RemoveAllInstancesOfGameplayTag self (Variables|InternalData|GetShieldedTags)))
          (:then_1
            (bind _returnvalue_2 (Class|FortCosmeticMaterialBlueprintLibrary|GetorCreateDynamicOverlayMaterial (Variables|Default|GetSkeletalMeshComponentBudgeted)))
            (Rendering|Material|SetScalarParameterValue _returnvalue_2 "EndTime" (Utilities|Time|GetGameTimeinSeconds) :!class /Script/Engine.MaterialInstanceDynamic)
            (CallFunction|ShieldDelayRemove))))))
  (CallFunction|ApplyChildWeakpointMaterials)
  (CallFunction|PlayWeakpointFadeOut true))

(fn ShouldUpdateWeakpoint
  (params (TargetWeakpointKind :type WeakpointKind%20Enum :default "NewEnumerator0"))
  (switch Utilities|FlowControl|Switch|SwitchonWeakpointKind TargetWeakpointKind
    (:NewEnumerator0
      _)
    (:NewEnumerator1
      _)
    (:NewEnumerator2
      _)
    (:NewEnumerator3
      (return true)))
  (bind _weakpointactor (CallFunction|GetChildWeakpointActorbyKind (Variables|GetTargetWeakpointKind)))
  (Utilities|Casting|CastToBGA_WeakpointActor_Large_Spire _weakpointactor
    (:then
      (return true))
    (:CastFailed
      (return false))))

(fn StartStateLogic
  (AI|Logic|StartLogic (Variables|Default|GetStateTree))
  (bind _returnvalue (Query|BPRegisterListener (Variables|Default|GetFortNearbyActorsPerception) (Utilities|Struct|MakeFortNearbyActorsPerceptionConfiguration (ScalableFloat|GetValueatLevel (Variables|InternalData|GetSF_DetectionUpdateRate)) 3 0 (ScalableFloat|GetValueatLevel (Variables|InternalData|GetSF_DetectionRadius)) (Utilities|Array|MakeArray "Hostile") false "(TokenStreamVersion=0,TagDictionary=((TagName=\"Athena.AI.IgnoreAll\")),QueryTokenStream=(0,1,3,1,0),UserDescription=\"\",AutoDescription=\" NONE( Athena.AI.IgnoreAll )\")") (EventDispatchers|CreateEvent :!event OnNearbyActorDetected)))
  (Variables|ForCreative|SetPerceptionListenerHandle _returnvalue))

(fn TestLineTraceAndSetStopPoint
  (params (Start :type Vector) (End :type Vector) (TraceChannel :type ETraceTypeQuery%20Enum :default "TraceTypeQuery1") (TraceComplex :type Boolean))
  (bind (_outhit _returnvalue) (Collision|LineTraceByChannel Start End TraceChannel TraceComplex (Utilities|Array|MakeArray (Variables|/Script/Engine.ChildActorComponent|GetChildActor (Variables|Default|GetWeakpoint01)) (Variables|/Script/Engine.ChildActorComponent|GetChildActor (Variables|Default|GetWeakpoint02)) (Variables|/Script/Engine.ChildActorComponent|GetChildActor (Variables|Default|GetWeakpoint03)) (Variables|/Script/Engine.ChildActorComponent|GetChildActor (Variables|Default|GetWeakpoint04)) self)))
  (if _returnvalue
    (if (< (.Distance (Collision|BreakHitResult _outhit)) (Variables|BeamData|GetCurrentClosestDistance))
      (Variables|BeamData|SetbFoundBlockingHit (Collision|BreakHitResult _outhit))
      (Variables|BeamData|SetCurrentClosestDistance (.Distance (Collision|BreakHitResult _outhit)))
      (Variables|BeamData|SetStopHitNormal (.Normal (Collision|BreakHitResult _outhit)))
      (Variables|BeamData|SetStopHitActor (.HitActor (Collision|BreakHitResult _outhit)))
      (bind (_returnvalue_roll _returnvalue_pitch _returnvalue_yaw) (Math|Rotator|BreakRotator (Math|Rotator|MakeRotationfromAxes (Transformation|GetActorForwardVector (.HitActor (Collision|BreakHitResult _outhit))) (Transformation|GetActorRightVector (.HitActor (Collision|BreakHitResult _outhit))) (.Normal (Collision|BreakHitResult _outhit)))))
      (Variables|SetTransformBeamStopPos_Location (Math|Transform|MakeTransform (.ImpactPoint (Collision|BreakHitResult _outhit)) (Math|Rotator|MakeRotator _returnvalue_roll _returnvalue_pitch _returnvalue_yaw)))
      (return (Variables|BeamData|GetbFoundBlockingHit))
      (else
        _)))
  (return false))

(fn ToggleWeakpointMaterialActive
  (params (WeakpointDataIndex :type Integer) (WeakpointToggled :type Boolean))
  (Rendering|Material|SetCustomPrimitiveDataFloat (CallFunction|GetSpireMesh) WeakpointDataIndex (Math|Float|SelectFloat 1.0 0.0 (and WeakpointToggled (not (Variables|InternalData|GetbShouldbeShielded))))) ; set health to 33%
)

(fn UpdateChildWeakpointActorByKind
  (params (TargetWeakpointKind :type WeakpointKind%20Enum :default "NewEnumerator0") (EnableWeakpoint :type Boolean))
  (bind _targetweakpointkind (Variables|GetTargetWeakpointKind))
  (if (CallFunction|ShouldUpdateWeakpoint TargetWeakpointKind)
    (CallFunction|UpdateChildWeakpointMaterials _targetweakpointkind EnableWeakpoint)
    (bind _weakpointactor (CallFunction|GetChildWeakpointActorbyKind _targetweakpointkind))
    (LocalVariables|SetTargetWeakpointActor _weakpointactor)
    (if (Variables|GetEnableWeakpoint)
      (Class|BGAWeakpointActorBase|EnableWeakpoint (LocalVariables|GetTargetWeakpointActor))
      (else
        (Class|BGAWeakpointActorBase|DisableWeakpoint (LocalVariables|GetTargetWeakpointActor))))))

(fn UpdateCurrentTarget
  (bind _targets (PawnTracking|GetSensedTargets))
  (bind _sf_targetupdatetimemin (Variables|InternalData|GetSF_TargetUpdateTimeMin))
  (bind _returnvalue (ScalableFloat|GetValueatLevel _sf_targetupdatetimemin))
  (bind _sf_targetupdatetimemax (Variables|InternalData|GetSF_TargetUpdateTimeMax))
  (bind _returnvalue_2 (ScalableFloat|GetValueatLevel _sf_targetupdatetimemax))
  (bind _returnvalue_3 (Math|Random|RandomFloatinRange _returnvalue _returnvalue_2))
  (Development|PrintString "@@@ UpdateCurrentTarget [BGA_Large_Spire]: CALLED!" false (Variables|Debug|GetbDebugActor))
  (if (not (Utilities|Array|IsEmpty _targets))
    (CallFunction|SetCurrentTargetPlayer (Utilities|Array|RandomArrayItem _targets))
    (bind _returnvalue_4 (Utilities|Time|SetTimerbyFunctionName 0 "UpdateCurrentTarget" _returnvalue_3))
    (Variables|InternalData|SetTargetUpdateTimerHandle _returnvalue_4)
    (Development|PrintString (Utilities|String|Append "@@@@@@ UpdateCurrentTarget [BGA_Large_Spire]: Next update in " (Utilities|String|ToString(Float) _returnvalue_3) " seconds!") false (Variables|Debug|GetbDebugActor))
    (else
      (Variables|InternalData|SetCurrentTargetPlayerPawn 0)
      (Utilities|Time|ClearandInvalidateTimerbyHandle (Variables|InternalData|GetTargetUpdateTimerHandle)) ; No need to keep firing delegate if there are no targets. Timer will be retriggered when new targets enter
    )))

(fn ConstructionScript
  (bind _spiremesh (CallFunction|GetSpireMesh))
  (Rendering|Material|SetCustomPrimitiveDataFloat (CallFunction|GetSpireMesh) 0 1.0) ; initalize the heath value on the skel mesh
  (Rendering|Material|SetCustomPrimitiveDataFloat _spiremesh 1 0.0) ; set health to 33%
  (Rendering|Material|SetCustomPrimitiveDataFloat _spiremesh 2 0.0) ; set health to 33%
  (Rendering|Material|SetCustomPrimitiveDataFloat _spiremesh 3 0.0) ; set health to 33%
  (Rendering|Material|SetCustomPrimitiveDataFloat _spiremesh 4 0.0) ; set health to 33%
  (Rendering|Material|SetCustomPrimitiveDataFloat _spiremesh 5 0.0) ; set health to 33%
)
