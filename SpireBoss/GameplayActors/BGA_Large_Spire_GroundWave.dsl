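; Large Spire boss ground-wave actor: an expanding donut shockwave that damages what it hits and destroys itself when its duration ends.
; BeginPlay plays a camera shake and sets wave duration from a scalable float.
; Size changes drive ring mesh scales, audio cylinder radius, Niagara wave size and a sine-based opacity on a dynamic material.
; Hits apply a global environment GameplayEffect, skipping Spire weakpoints and ignored actors.
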(event EventBeginPlay
  (Camera|PlayWorldCameraShake "/SpireSharedAssets/VFX/Slam/B_SpireSlamShockWave_CameraShake.B_SpireSlamShockWave_CameraShake_C" (Transformation|GetActorLocation) (ScalableFloat|GetValueatLevel (Variables|Internal|GetSF_CameraShakeRadius)) (ScalableFloat|GetValueatLevel (Variables|Internal|GetSF_CameraShakeRadius)))
  (Variables|Default|SetGroundWaveDuration (ScalableFloat|GetValueatLevelwithContextParam (Actor|GetOwner) (Variables|Default|GetSF_GroundWaveDuration))))

(event Custom|HandleShockwaveDamage
  (params (TargetActor :type Actor%20Object%20Reference))
  (bind _asfort_pawn (Utilities|Casting|CastToFortPawn TargetActor)
    (:then
      (if (GameplayTags|HasAnyMatchingGameplayTags _asfort_pawn (Variables|Internal|GetPawnShockwaveExclusionTags))
        _
        (else
          _)))
    (:CastFailed
      _))
  (CallFunction|ApplyGroundWaveDamage TargetActor))

(event OnActorHit(FortDonutAttack)
  (params (HitActor :type Actor%20Object%20Reference))
  (Utilities|Casting|CastToBGA_WeakpointActor_Large_Spire HitActor
    (:CastFailed
      (if (not (Utilities|Array|ContainsItem (Variables|Internal|GetWaveIgnoreActors) HitActor))
        (CallFunction|HandleShockwaveDamage HitActor)))))

(event OnDurationEnd(FortDonutAttack)
  (Rendering|SetVisibility (Variables|Default|GetSM_Spire_Outter_Ring) false false)
  (Actor|DestroyActor))

(event OnSizeChange(FortDonutAttack)
  (params (OuterRadius :type Float%20%28single-precision%29) (InnerRadius :type Float%20%28single-precision%29) (Height :type Float%20%28single-precision%29) (DurationNormalized :type Float%20%28single-precision%29))
  (bind _returnvalue (/ Height 1e+02))
  (bind _returnvalue_2 (/ OuterRadius 1e+02))
  (Transformation|SetWorldScale3D (Variables|Default|GetSM_Spire_Outter_Ring) (Math|Vector|MakeVector _returnvalue_2 _returnvalue_2 _returnvalue))
  (Transformation|SetWorldScale3D (Variables|Default|GetSM_Spire_Inner_Ring) (Math|Vector|MakeVector (/ InnerRadius 1e+02) (/ InnerRadius 1e+02) _returnvalue))
  (Rendering|Material|SetScalarParameterValue (Variables|Default|GetShockwaveMaterial) "OpacityMult" (Math|Trig|Sin(Radians) (* DurationNormalized (Variables|Default|GetGroundWaveDuration))) :!class /Script/Engine.MaterialInstanceDynamic)
  (Control|SetRadius (Variables|Default|GetAudioShapeCylinder) OuterRadius :!class /Script/AudioShapes.AudioShapeCylinderComponent)
  (Niagara|SetNiagaraVariable(Float) (Variables|Default|GetNS_Spire_Slam_Wave) "TargetWaveSize" _returnvalue_2)
  (Utilities|FlowControl|DoOnce false
    (:Completed
      (Components|Activation|SetActive (Variables|Default|GetNS_Spire_Slam_Wave) true true))))

(fn ApplyGroundWaveDamage
  (params (EffectTargetActor :type Actor%20Object%20Reference))
  (bind _returnvalue (Environment|ApplyGlobalEnvironmentGameplayEffecttoActor EffectTargetActor (Variables|Internal|GetGE_GroundWaveDamage)))
  (return _returnvalue))

(fn ConstructionScript
  (bind _returnvalue (Rendering|Material|CreateDynamicMaterialInstance "/SpireBoss/VFX/MI_Spire_Shockwave_Gradient_Glow.MI_Spire_Shockwave_Gradient_Glow" "None" "None" :!class /Script/Engine.KismetMaterialLibrary))
  (bind _output_get (Variables|Default|SetShockwaveMaterial _returnvalue))
  (Rendering|Material|SetMaterial (Variables|Default|GetSM_Spire_Outter_Ring) 0 _output_get))
