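; BGA_Large_Spire_Proj_Context is a SpireBoss gameplay actor, by its name a context actor for a large spire projectile.
; BeginPlay only calls a debug PrintString reporting that it ran, with screen printing off and log printing driven by the bDebugActor variable.
; Overlap, Tick and the construction script are empty, so the file itself holds no projectile logic.
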(event EventBeginPlay
  (Development|PrintString "@@@@@@@@@ Event BeginPlay [BGA_Large_Spire_Proj_Context]: CALLED!" false (Variables|Debug|GetbDebugActor)))

(event Collision|EventActorBeginOverlap
  (params (OtherActor :type Actor%20Object%20Reference)))

(event EventTick
  (params (DeltaSeconds :type Float%20%28single-precision%29)))

(fn ConstructionScript)
