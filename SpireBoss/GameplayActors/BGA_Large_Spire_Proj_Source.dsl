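; Projectile source actor for the large Spire boss that plays spawn-in effects, then fires a fan of homing projectiles when triggered.
; InitiateProjectileLaunch precomputes fan transforms around an offset center, shuffles them and starts a looping timer that spawns one projectile per tick.
; Each shot picks a target from the owning Spire's sensed pawns using avoid-tag, yaw-angle and already-targeted checks, and carries an explosion effect container.
; After the last shot it destroys itself, deactivating its Niagara effect.
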(event EventBeginPlay
  (bind (_returnvalue_location _returnvalue_rotation) (Math|Transform|BreakTransform (Transformation|GetActorTransform)))
  (bind (_x _y _z) (Math|Rotator|GetAxes _returnvalue_rotation))
  (Development|PrintString "@@@@@@@@@ Event BeginPlay [BGA_Large_Spire_Proj_Source]: CALLED!" false (Variables|debug|GetbDebugActor))
  (Game|BindEventtoOnDestroyed (delegate Custom|Event%20OnDestroyed))
  (Variables|Internal|SetbIsSourceTriggered false)
  (Utilities|Array|Clear (Variables|Internal|GetSourceShotPlayerPawnArray))
  (Variables|Internal|SetTargetPlayer 0)
  (Utilities|Array|Clear (Variables|Internal|GetProjectileTransformArray))
  (Utilities|FlowControl|Delay 1.0)
  (bind _returnvalue (Niagara|SpawnSystematLocation "/SpireSharedAssets/VFX/NotCookContent/NS_Spire_Prj_SpawnIn.NS_Spire_Prj_SpawnIn" _returnvalue_location (Math|Rotator|MakeRotfromXY _y _z) "1.000000,1.000000,1.000000" false))
  (Variables|Default|SetSpawnedFX _returnvalue)
  (Audio|PlaySoundatLocation "/SpireBoss/Audio/MetaSounds/MSS_MajorSpire_Proj_SpawnIn.MSS_MajorSpire_Proj_SpawnIn" _returnvalue_location))

(event Custom|Event%20OnDestroyed
  (params (DestroyedActor :type Actor%20Object%20Reference))
  (Development|PrintString "@@@@@@@@@ Event OnDestroyed [BGA_Large_Spire_Proj_Source]: CALLED!" false (Variables|debug|GetbDebugActor))
  (CallFunction|DeactivateSpawnSystemParticles))

(event Custom|DeactivateSpawnSystemParticles
  (bind _spawnedfx (Variables|Default|GetSpawnedFX))
  (Networking|CanExecuteCosmeticEvents
    (:True
      (Niagara|SetNiagaraVariable(Bool) _spawnedfx "Die" true)
      (Components|Activation|Deactivate _spawnedfx))))

(event Custom|InitiateProjectileLaunch
  (params (FireAbility :type Fort%20Gameplay%20Ability%20Object%20Reference))
  (bind _self self)
  (bind (_returnvalue_location _returnvalue_rotation) (Math|Transform|BreakTransform (Transformation|GetActorTransform _self)))
  (bind (_fireoffset_x _fireoffset_y _fireoffset_z) (Math|Vector|BreakVector (Variables|GetFireOffset_X)))
  (Variables|Internal|SetFireAbilityRef FireAbility)
  (Variables|Internal|SetFanCenterRotator _returnvalue_rotation)
  (Variables|Internal|SetbIsSourceTriggered true)
  (bind _output_get (Variables|Internal|SetFanCenterLoc (Utilities|Operators|Add _returnvalue_location (* (Math|Vector|GetForwardVector _returnvalue_rotation) _fireoffset_x) (* (Math|Vector|GetRightVector _returnvalue_rotation) _fireoffset_y) (* (Math|Vector|GetUpVector _returnvalue_rotation) _fireoffset_z))))
  (if (Variables|debug|GetbDebugActor)
    (Rendering|Debug|DrawDebugSphere _output_get 5e+01 12 "(R=1.000000,G=0.027003,B=0.636845,A=1.000000)" 1e+01 2.0))
  (Variables|Internal|SetAngle (ScalableFloat|GetValueatLevel (Variables|Tunable|GetSF_SourceFireAngle)))
  (Variables|Internal|SetProjectileCount (CallFunction|GetProjectileCount))
  (Variables|Internal|SetRadius (ScalableFloat|GetValueatLevel (Variables|Tunable|GetSF_SourceFireRadius)))
  (Variables|Internal|SetIncrementAngle (/ (Variables|Internal|GetAngle) (Variables|Internal|GetProjectileCount)))
  (Variables|Internal|SetIncrementRotator (/ (Variables|Internal|GetIncrementAngle) 2))
  (if (Variables|debug|GetbDebugActor)
    (Rendering|Debug|DrawDebugLine (Variables|Internal|GetFanCenterLoc) (+ (Variables|Internal|GetFanCenterLoc) (* (Math|Vector|GetRightVector (Variables|Internal|GetFanCenterRotator)) 5e+02)) "(R=0.458637,G=0.025770,B=0.989583,A=1.000000)" 1e+01 3.0)
    (else
      _))
  (for _index (range 1 (+ (Variables|Internal|GetProjectileCount) 1))
    (bind _output_get_2 (Variables|SetCurrentProjectileTransform_Location (Math|Transform|MakeTransform (+ (Variables|Internal|GetFanCenterLoc) (* (Math|Vector|RotateVectorAroundAxis (Math|Vector|GetRightVector (Variables|Internal|GetFanCenterRotator)) (Utilities|Operators|Subtract (* (Variables|Internal|GetIncrementAngle) _index) (Variables|Internal|GetIncrementRotator) (/ (Variables|Internal|GetAngle) 2)) "0.000000,0.000000,1.000000") (Variables|Internal|GetRadius))) (Math|Rotator|MakeRotfromX (Math|Vector|RotateVectorAroundAxis (Math|Vector|GetRightVector (Variables|Internal|GetFanCenterRotator)) (Utilities|Operators|Subtract (* (Variables|Internal|GetIncrementAngle) _index) (Variables|Internal|GetIncrementRotator) (/ (Variables|Internal|GetAngle) 2)) "0.000000,0.000000,1.000000")))))
    (bind _output_get_location (Math|Transform|BreakTransform _output_get_2))
    (if (Variables|debug|GetbDebugActor)
      (Rendering|Debug|DrawDebugSphere _output_get_location 5e+01 12 "(R=0.005771,G=0.489583,B=0.340933,A=1.000000)" 1e+01 2.0))
    (Utilities|Array|AddUnique (Variables|Internal|GetProjectileTransformArray) (Variables|Internal|GetCurrentProjectileTransform)))
  (Utilities|Array|Shuffle (Variables|Internal|GetProjectileTransformArray))
  (Variables|Internal|SetProjectileCounter 0)
  (bind _returnvalue (Utilities|Time|SetTimerbyEvent (delegate Custom|SpawnNextProjectile) (ScalableFloat|GetValueatLevel (Variables|Tunable|GetSF_SourceFireDelay)) true))
  (Variables|Internal|SetProjectileTimerHandle _returnvalue))

(event Custom|UpdateFX
  (params (FXLocation :type Vector))
  (bind _spawnedfx (Variables|Default|GetSpawnedFX))
  ; fx - @philippe patient if something's off
  (Niagara|SetNiagaraVariable(Bool) _spawnedfx "bisShootingProjectile" true)
  (if (not (Variables|Default|GetbSpawnedBurstEffect))
    (Niagara|SpawnSystematLocation "/SpireSharedAssets/VFX/NotCookContent/NS_Spire_Prj_BurstPrj.NS_Spire_Prj_BurstPrj" FXLocation)
    (Variables|Default|SetbSpawnedBurstEffect true))
  (Utilities|FlowControl|Delay 0.1)
  (Niagara|SetNiagaraVariable(Bool) _spawnedfx "bisShootingProjectile" false))

(event Custom|SpawnNextProjectile
  (bind (_output_location _output_rotation) (Math|Transform|BreakTransform (Utilities|Array|Get(acopy) (Variables|Internal|GetProjectileTransformArray) (Variables|Internal|GetProjectileCounter))))
  (bind _output_location_2 (Math|Transform|BreakTransform (Utilities|Array|Get(acopy) (Variables|Internal|GetProjectileTransformArray) (Variables|Internal|GetProjectileCounter))))
  (bind _spireactor (Variables|Internal|GetSpireActor))
  (bind _sensedplayerpawns (Variables|/SpireBoss/GameplayActors/BGA_Large_Spire.BGA_Large_Spire_C|GetSensedPlayerPawns _spireactor))
  (bind _targetplayer (Variables|Internal|GetTargetPlayer))
  (if (Variables|debug|GetbDebugActor)
    (Rendering|Debug|DrawDebugLine _output_location (+ _output_location (* (Math|Vector|GetForwardVector _output_rotation) 5e+02)) "(R=0.890625,G=0.008859,B=0.000000,A=1.000000)" 1e+01 3.0))
  (CallFunction|UpdateFX _output_location_2)
  (Variables|Internal|SetbFoundNewTarget false)
  (Variables|Internal|SetTargetPlayer 0)
  (if (> (Utilities|Array|Length _sensedplayerpawns) 0)
    (bind _array_element (Utilities|Array|ForEachLoopwithBreak _sensedplayerpawns)
      (:LoopBody
        (Variables|Internal|SetTargetPlayer _array_element)
        (if (GameplayTags|HasAnyMatchingGameplayTags _array_element (Variables|Internal|GetAvoidDetectionTagsContainer))
          _
          (elif (Utilities|Array|ContainsItem (Variables|Internal|GetSourceShotPlayerPawnArray) _targetplayer)
            _
            (elif (CallFunction|IsTargetHomingValid _targetplayer (Utilities|Array|Get(acopy) (Variables|Internal|GetProjectileTransformArray) (Variables|Internal|GetProjectileCounter)))
              (Variables|Internal|SetbFoundNewTarget true)
              (Utilities|Array|Add (Variables|Internal|GetSourceShotPlayerPawnArray) (Variables|Internal|GetTargetPlayer))))))
      (:Completed
        _))
    (if (Variables|Internal|GetbFoundNewTarget)
      _
      (elif (> (Utilities|Array|Length (Variables|/SpireBoss/GameplayActors/BGA_Large_Spire.BGA_Large_Spire_C|GetSensedPlayerPawns (Variables|Internal|GetSpireActor))) 0)
        (Variables|Internal|SetTargetPlayer 0)
        (bind _array_element_2 (Utilities|Array|ForEachLoopwithBreak (Variables|/SpireBoss/GameplayActors/BGA_Large_Spire.BGA_Large_Spire_C|GetSensedPlayerPawns (Variables|Internal|GetSpireActor)))
          (:LoopBody
            (if (and (not (GameplayTags|HasAnyMatchingGameplayTags _array_element_2 (Variables|Internal|GetAvoidDetectionTagsContainer))) (CallFunction|IsTargetHomingValid _array_element_2 (Utilities|Array|Get(acopy) (Variables|Internal|GetProjectileTransformArray) (Variables|Internal|GetProjectileCounter))))
              (Variables|Internal|SetTargetPlayer _array_element_2)))
          (:Completed
            _))
        (else
          _)))
    (CallFunction|HandleProjectileSpawnwithParams (Variables|Internal|GetTargetPlayer))
    (CallFunction|IncrementProjectileSpawnCounter)
    (else
      (CallFunction|HandleProjectileSpawnwithParams 0)
      (CallFunction|IncrementProjectileSpawnCounter))))

(event Custom|IncrementProjectileSpawnCounter
  (bind _result (Math|Integer|IncrementInt (Variables|Internal|GetProjectileCounter))
    (:%20%20
      (if (>= _result (Variables|Internal|GetProjectileCount))
        (CallFunction|OnFinishedSpawningProjectiles)))))

(event Custom|HandleProjectileSpawn
  (params (Owner :type Actor%20Object%20Reference) (Spawn%20Location :type Vector) (Spawn%20Rotation :type Rotator) (ProjHomingTarget :type Actor%20Object%20Reference))
  (bind _projectilefanproj (Variables|Internal|GetProjectileFanProj))
  (bind _gecontainerspec (CallFunction|MakeProjEffectContainertoApplyonExplode Owner))
  (bind _returnvalue (Projectile|SpawnProjectilewithParams (Variables|Default|GetProjectile%20Class) Owner (Utilities|Struct|MakeSpawnProjectileParams Spawn%20Location Spawn%20Rotation Spawn%20Rotation "(TargetSelection=(List=,bStopAtFirstSuccess=True,bKeepCheckingListOnIndestructibleHit=False,AbilityTargetBlockedGameplayTags=(GameplayTags=),bUseWeaponRanges=True,bUseMaxYawAngleToTarget=False,MaxYawAngleToTarget=0.000000),TargetGameplayEffectSpecs=,OwnerGameplayEffectSpecs=,ActivationCues=(GameplayTags=),ImpactCues=(GameplayTags=),ImpactNoiseRange=0.000000,FlyByNoiseRange=0.000000,bOverrideChargeMagnitude=False,ChargeMagnitudeOverrideValue=0.000000)" _gecontainerspec false 0 (Math|Random|RandomFloatinRange (ScalableFloat|GetValueatLevel (Variables|Tunable|GetSF_ProjectileInitialSpeedMin)) (ScalableFloat|GetValueatLevel (Variables|Tunable|GetSF_ProjectileInitialSpeedMax))) (Math|Random|RandomFloatinRange (ScalableFloat|GetValueatLevel (Variables|Tunable|GetSF_ProjectileInitialSpeedMin)) (ScalableFloat|GetValueatLevel (Variables|Tunable|GetSF_ProjectileInitialSpeedMax))) 0.0)))
  (Utilities|IsValid _returnvalue
    (:Is%20Valid
      (bind _asb_prj_spire_boss (Utilities|Casting|CastToB_Prj_Spire_Boss _returnvalue))
      (bind _output_get (Variables|Internal|SetProjectileFanProj _asb_prj_spire_boss))
      (Variables|/SpireBoss/GameplayActors/B_Prj_Spire_Boss.B_Prj_Spire_Boss_C|SetFireAbilityRef (Variables|Internal|GetFireAbilityRef) _output_get)
      (Utilities|IsValid ProjHomingTarget
        (:Is%20Valid
          (Variables|/SpireBoss/GameplayActors/B_Prj_FirePetal_Boss_MaskFire_Prj.B_Prj_FirePetal_Boss_MaskFire_Prj_C|SetProjHomingTarget ProjHomingTarget _projectilefanproj))
        (:Is%20Not%20Valid
          (Variables|/SpireBoss/GameplayActors/B_Prj_FirePetal_Boss_MaskFire_Prj.B_Prj_FirePetal_Boss_MaskFire_Prj_C|SetProjHomingTarget 0 _projectilefanproj))))))

(event Custom|HandleProjectileSpawnWithParams
  (params (ProjHomingTarget :type Actor%20Object%20Reference))
  (bind (_output_location _output_rotation) (Math|Transform|BreakTransform (Utilities|Array|Get(acopy) (Variables|Internal|GetProjectileTransformArray) (Variables|Internal|GetProjectileCounter))))
  (CallFunction|HandleProjectileSpawn (Actor|GetOwner self) _output_location _output_rotation ProjHomingTarget))

(fn DebugPrintArray
  (params (PlayerArray :type Array%20of%20Fort%20Player%20Pawn%20Object%20References))
  (Development|PrintString (Utilities|String|Append "@@@@@@@@@ DebugPrintArray [GA_Large_Spire_Proj_Source]: CALLED with a Player Array of size " (Utilities|String|ToString(Integer) (Utilities|Array|Length (Variables|GetPlayerArray))) "!") false (Variables|debug|GetbDebugActor))
  (for (_array_element _array_index) (Variables|GetPlayerArray)
    (Development|PrintString (Utilities|String|Append "@@@@@@@@@@@@ DebugPrintArray [GA_Large_Spire_Proj_Source]:  [" (Utilities|String|ToString(Integer) _array_index) "] = " (Utilities|GetDisplayName _array_element)) false (Variables|debug|GetbDebugActor))))

(fn GetProjectileCount
  (return (ScalableFloat|GetValueasInteger (Variables|Tunable|GetSF_SourceProjectileCount))))

(fn GetYawToPlayerPawn
  (params (PlayerPawn :type Pawn%20Object%20Reference) (SourceTransform :type Transform) (GetAbsValue :type Boolean))
  (bind (_returnvalue_roll _returnvalue_pitch _returnvalue_yaw) (Math|Rotator|BreakRotator (Math|Rotator|FindRelativeLookatRotation SourceTransform (Pawn|GetNavAgentLocation PlayerPawn))))
  (bind _returnvalue (Pawn|GetNavAgentLocation PlayerPawn))
  (if (Variables|debug|GetbDebugActor)
    (Rendering|Debug|DrawDebugSphere _returnvalue 5e+01 12 "(R=0.942708,G=0.018629,B=0.014127,A=1.000000)" 1e+01 2.5))
  (if (Variables|GetGetAbsValue)
    (return (Math|Float|Absolute(Float) _returnvalue_yaw))
    (else
      (return _returnvalue_yaw))))

(fn IsTargetHomingValid
  (params (PlayerPawn :type Fort%20Player%20Pawn%20Object%20Reference) (SourceTransform :type Transform))
  (return (<= (CallFunction|GetYawtoPlayerPawn PlayerPawn SourceTransform true) (ScalableFloat|GetValueatLevel (Variables|Tunable|GetSF_SourceForwardFacingAngle)))))

(fn MakeProjEffectContainerToApplyOnExplode
  (params (Owner :type Actor%20Object%20Reference))
  (bind _returnvalue (Ability|MakeGameplayEffectContainerSpecfromAbilityComponent (Variables|Internal|GetEffectContainerExplode) (Ability|GetAbilitySystemComponent Owner)))
  (return _returnvalue))

(fn OnFinishedSpawningProjectiles
  (Utilities|Time|ClearandInvalidateTimerbyHandle (Variables|Internal|GetProjectileTimerHandle))
  (Actor|DestroyActor self))

(fn SetOwningSpire
  (params (Spire :type BGA%20Large%20Spire%20Object%20Reference))
  (Variables|Internal|SetSpireActor Spire))

(fn ConstructionScript)
