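; Core weakpoint actor for the large Spire boss: a sphere that multiplies hit damage while enabled and ricochets hits with zero damage while disabled.
; BeginPlay registers it as an aim-assist target, configures sphere collision, and sets overlap generation from a hotfix value once playlist data is ready.
; The server caches the parent Spire as owner, whose shielded tags select the ricochet effect.
; Enable and disable toggle collision, and only enabled weakpoints accept auto-fire.
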(event EventEndPlay
  (params (EndPlayReason :type EEndPlayReason%20Enum :default "Destroyed"))
  (AimAssist|UnregisterAimAssistTarget 0))

(event EventBeginPlay
  (bind _weakpointsphere (Variables|Default|GetWeakpointSphere))
  (|CallParentFunction)
  (AimAssist|RegisterAimAssistTarget 0 (Variables|Default|GetWeakpointSphere))
  (Utilities|FlowControl|Sequence
    (:then_0
      (if (and (ScalableFloat|GetValueasBool (Variables|InternalData|GetSF_UseTempClientOverrideFlow)) (Editor|InEditororPIE))
        (if (Networking|IsServer)
          (Variables|InternalData|SetbIsVerified true)
          (elif (Variables|InternalData|GetbIsVerified)
            _))
        (CallFunction|SetWeakpointCollisionKind "NewEnumerator2" 0 0 (Variables|Default|GetWeakpointSphere))
        (CallFunction|UpdateWeakpointCollision false "NewEnumerator2" (Variables|Default|GetWeakpointSphere))
        (else
          (CallFunction|SetWeakpointCollisionKind "NewEnumerator2" 0 0 _weakpointsphere)
          (CallFunction|UpdateWeakpointCollision false "NewEnumerator2" _weakpointsphere))))
    (:then_1
      (Class|FortAsyncActionCurrentPlaylistReady|PlaylistDataReadyAsync
        (:OnReady
          (Variables|/Script/Engine.PrimitiveComponent|SetbGenerateOverlapEvents (ScalableFloat|GetValueasBool (Variables|Hotfix|GetHF_EnableOverlaps)) (Variables|Default|GetWeakpointSphere)))))))

(event EventOnWorldReady
  (if (Networking|IsServer)
    (bind _asbga_large_spire (Utilities|Casting|CastToBGA_Large_Spire (Actor|GetParentActor self)))
    (Variables|InternalData|SetSpireOwner _asbga_large_spire)))

(event EventPlayRicochet
  (params (Hit%20Result :type Hit%20Result%20Structure))
  (if (GameplayTags|HasAllMatchingGameplayTags (Variables|InternalData|GetSpireOwner) (Variables|InternalData|GetShieldedTags))
    (Variables|Tunable|SetWeakpoint_RicochetFX "/SpireSharedAssets/VFX/Temp/NS_MinorSpire_Ricochet.NS_MinorSpire_Ricochet")
    (else
      (Variables|Tunable|SetWeakpoint_RicochetFX "/SpireSharedAssets/VFX/NS_SpireBoss_RicochetImpact.NS_SpireBoss_RicochetImpact")))
  (|CallParentFunction Hit%20Result))

(fn DisableWeakpoint
  (if (Variables|Internal|GetbIsEnabled)
    (Variables|Internal|SetbCollisionEnabled false)
    (|CallParentFunction)))

(fn EnableWeakpoint
  (if (not (Variables|Internal|GetbIsEnabled))
    (Variables|Internal|SetbCollisionEnabled true)
    (|CallParentFunction)))

(fn IsValidAutoFireTarget
  (params (Attacker :type Actor%20Object%20Reference))
  (return (Variables|Internal|GetbIsEnabled)))

(fn ProcessHitResult
  (params (Damage :type Float%20%28double-precision%29) (HitResult :type Hit%20Result%20Structure))
  (if (Variables|Internal|GetbIsEnabled)
    (return (* Damage (ScalableFloat|GetValueatLevel (Variables|InternalData|GetSF_CriticalDamageMultiplier))))
    (else
      (CallFunction|PlayRicochet HitResult)
      (return 0.0))))

(fn ConstructionScript
  (|CallParentFunction))
