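; Weakpoint actor on the large Spire boss: a child-actor critical hit target that registers for aim assist and auto-fire.
; The server sets its health by playlist team size and activates it once the parent spire's plunge finishes.
; Hits ricochet with no damage while the spire has shield tags; otherwise damage is multiplied when enabled or divided when disabled.
; Depleted health deactivates it via replicated state, updating the spire, playing break effects and optionally dropping ammo loot at its appendage bone.
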(event EventEndPlay
  (params (EndPlayReason :type EEndPlayReason%20Enum :default "Destroyed"))
  (AimAssist|UnregisterAimAssistTarget 0))

(event EventBeginPlay
  (bind _weakpointcapsule (Variables|Default|GetWeakpointCapsule))
  (|CallParentFunction)
  (CallFunction|SetWeakpointCollisionKind "NewEnumerator1" 0 _weakpointcapsule)
  (CallFunction|UpdateWeakpointCollision true "NewEnumerator1" _weakpointcapsule)
  (AimAssist|RegisterAimAssistTarget 0 _weakpointcapsule)
  (bind (_gamestate _playlist) (Class|FortAsyncActionCurrentPlaylistReady|PlaylistDataReadyAsync)
    (:OnReady
      (Variables|/Script/Engine.PrimitiveComponent|SetbGenerateOverlapEvents (ScalableFloat|GetValueasBool (Variables|Hotfix|GetHF_EnableOverlaps)) (Variables|Default|GetWeakpointCapsule))
      (if (Networking|IsServer)
        (Variables|Internal|SetWeakpoint_CurrentHealth (ScalableFloat|GetValueatLevel (Variables|Internal|GetSF_PlaylistData) (Math|Conversions|ToFloat(Integer) (Variables|/Script/FortniteGame.FortPlaylist|GetMaxTeamSize _playlist))))
        (bind _asbga_large_spire (Utilities|Casting|CastToBGA_Large_Spire (Actor|GetParentActor self)))
        (bind _output_get (Variables|Internal|SetSpireOwner _asbga_large_spire))
        (Default|BindEventtoOnPlungeFinished _output_get (EventDispatchers|CreateEvent :!event Custom|OnSpireDescentComplete))))))

(event Custom|OnSpireDescentComplete
  (Variables|Internal|SetbForcedActive true))

(event EventPlayRicochet
  (params (Hit%20Result :type Hit%20Result%20Structure))
  (if (GameplayTags|HasAllMatchingGameplayTags (Variables|Internal|GetSpireOwner) (Variables|Internal|GetShieldedTags))
    (Variables|Tunable|SetWeakpoint_RicochetFX "/SpireSharedAssets/VFX/Temp/NS_MinorSpire_Ricochet.NS_MinorSpire_Ricochet")
    (else
      (Variables|Tunable|SetWeakpoint_RicochetFX "/SpireSharedAssets/VFX/NS_SpireBoss_RicochetImpact.NS_SpireBoss_RicochetImpact")))
  (|CallParentFunction Hit%20Result))

(event EventOnDamagePlayEffects
  (params (Damage :type Float%20%28single-precision%29) (DamageTags :type Gameplay%20Tag%20Container%20Structure%20%28by%20ref%29) (Momentum :type Vector :default "0, 0, 0") (HitInfo :type Hit%20Result%20Structure%20%28by%20ref%29) (InstigatedBy :type Fort%20Pawn%20Object%20Reference) (DamageCauser :type Actor%20Object%20Reference) (EffectContext :type Gameplay%20Effect%20Context%20Handle%20Structure))
  (if (GameplayTags|HasAllMatchingGameplayTags (Variables|Internal|GetSpireOwner) (Variables|Internal|GetShieldedTags))
    (Variables|Internal|SetInstigatorPawn InstigatedBy)
    (CallFunction|PlayRicochet HitInfo)
    (else
      (|CallParentFunction Damage DamageTags Momentum HitInfo InstigatedBy DamageCauser EffectContext))))

(fn GetAppendageWeakpointTransform
  (params (SpieAppendageKind :type SpireAppendageKind%20Enum :default "NewEnumerator0"))
  (bind _asbga_large_spire (Utilities|Casting|CastToBGA_Large_Spire (Actor|GetParentActor self))
    (:then
      (bind (_targetstructdata_spireappendagekind_2_c321b3314c9291adb879629b5339498e _targetstructdata_bonename_5_38cd39514d738e05454978b3e574f8a0) (Utilities|Struct|BreakSpireAppendageStruct (Class|BGALargeSpire|GetAppendageDatabyKind _asbga_large_spire SpieAppendageKind)))
      (bind (_returnvalue_location _returnvalue_rotation _returnvalue_scale) (Math|Transform|BreakTransform (Transformation|GetBoneTransform (Variables|/SpireBoss/GameplayActors/BGA_Large_Spire.BGA_Large_Spire_C|GetSkeletalMeshComponentBudgeted _asbga_large_spire) _targetstructdata_bonename_5_38cd39514d738e05454978b3e574f8a0 "RTS_World")))
      (return (Math|Transform|MakeTransform _returnvalue_location _returnvalue_rotation _returnvalue_scale)))
    (:CastFailed
      (return 0))))

(fn IsValidAutoFireTarget
  (params (Attacker :type Actor%20Object%20Reference))
  ; No need to check for Enabled, we want to allow Autofire even if not enabled.
  (return true))

(fn OnRep_bForcedActive
  (if (not (Variables|Internal|GetbForcedActive))
    (Components|Activation|Activate (Variables|Default|GetNS_AOEPulse_Break) true)
    (Audio|PlaySoundatLocation "/SpireSharedAssets/Audio/Cues/SC_MajorSpire_Weakpoint_Destroyed.SC_MajorSpire_Weakpoint_Destroyed" (Transformation|GetActorLocation)))
  (if (Networking|IsServer)
    (bind _asbga_large_spire (Utilities|Casting|CastToBGA_Large_Spire (Actor|GetParentActor self)))
    (bind _weakpointkind (Class|BGALargeSpire|GetWeakpointKindbyActor _asbga_large_spire self))
    (Class|BGALargeSpire|UpdateChildWeakpointActorbyKind _asbga_large_spire _weakpointkind (Variables|Internal|GetbForcedActive))))

(fn ProcessHealthDepleted
  (if (Variables|Internal|GetbForcedActive)
    (Development|PrintString (Utilities|String|Append "@@@ ProcessHealthDepleted [BGA_WeakpointActor_Base]: CALLED for " (Utilities|GetDisplayName self) "!") false (Variables|Debug|GetbDebugWeakpoint))
    (Variables|Internal|SetbForcedActive false)
    (if (ScalableFloat|GetValueasBool (Variables|Internal|GetSF_DropLootOnBreak))
      (bind _asbga_large_spire (Utilities|Casting|CastToBGA_Large_Spire (Actor|GetParentActor self)))
      (bind _output_get (LocalVariables|SetSpireActor _asbga_large_spire))
      (bind _weakpointkind (Class|BGALargeSpire|GetWeakpointKindbyActor _output_get self))
      (switch Utilities|FlowControl|Switch|SwitchonWeakpointKind _weakpointkind
        (:NewEnumerator0
          (bind _output_get_2 (LocalVariables|SetSpireAppendageKind "NewEnumerator0"))
          (bind _weakpointtransform (CallFunction|GetAppendageWeakpointTransform _output_get_2))
          (LocalVariables|SetWeakpointBoneTransform _weakpointtransform))
        (:NewEnumerator1
          (bind _output_get_3 (LocalVariables|SetSpireAppendageKind "NewEnumerator1"))
          (bind _weakpointtransform_2 (CallFunction|GetAppendageWeakpointTransform _output_get_3))
          (LocalVariables|SetWeakpointBoneTransform _weakpointtransform_2))
        (:NewEnumerator2
          (bind _output_get_4 (LocalVariables|SetSpireAppendageKind "NewEnumerator2"))
          (bind _weakpointtransform_3 (CallFunction|GetAppendageWeakpointTransform _output_get_4))
          (LocalVariables|SetWeakpointBoneTransform _weakpointtransform_3)))
      (Class|AthenaFunctionLibrary|SpawnLootfromTableUsingItemEntry "Loot_Spire_Ammo" 1 -1 (.location (LocalVariables|GetWeakpointBoneTransform)) (Math|Vector|GetForwardVector (.rotation (LocalVariables|GetWeakpointBoneTransform))) 0 true true false 0 "Other" "LootDrop"))))

(fn ProcessHitResult
  (params (Damage :type Float%20%28double-precision%29) (HitResult :type Hit%20Result%20Structure))
  (if (GameplayTags|HasAllMatchingGameplayTags (Variables|Internal|GetSpireOwner) (Variables|Internal|GetShieldedTags))
    (CallFunction|PlayRicochet HitResult)
    (return 0.0)
    (elif (Variables|Internal|GetbIsEnabled)
      (return (* Damage (ScalableFloat|GetValueatLevel (Variables|Internal|GetSF_CriticalDamageMultiplier))))
      (else
        (CallFunction|PlayRicochet HitResult)
        (return (/ Damage (ScalableFloat|GetValueatLevel (Variables|Internal|GetSF_BaseDamageDivider))))))))

(fn ShouldDamageWeakpoint
  (return (and (Variables|Tunable|GetbWeakpointCanBeDamaged) (Variables|Internal|GetbForcedActive))))

(fn UpdateWeakpointCollision
  (params (bEnableCollision :type Boolean) (CollisionKind :type Enum_WeakpointActor_CollisionKind%20Enum :default "NewEnumerator0") (CollisionComponent :type Primitive%20Component%20Object%20Reference))
  (bind _weakpointcapsule (Variables|Default|GetWeakpointCapsule))
  (|CallParentFunction bEnableCollision CollisionKind CollisionComponent)
  (Collision|SetCollisionResponsetoChannel _weakpointcapsule "ECC_Pawn" (select bEnableCollision "ECR_Block" "ECR_Ignore"))
  (Collision|SetCollisionResponsetoChannel _weakpointcapsule "ECC_Vehicle" (select bEnableCollision "ECR_Block" "ECR_Ignore")))

(fn ConstructionScript
  (|CallParentFunction))
