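; Weakpoint actor for the Minor Spire in the Spire boss content, serving as an aim-assist and auto-fire target.
; On BeginPlay it runs the parent logic, sets and enables collision on its weakpoint capsule, then registers that capsule for aim assist.
; It unregisters from aim assist on EndPlay.
; It reports itself as a valid auto-fire target only while its enabled flag is set, regardless of the attacker.
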(event EventBeginPlay
  (bind _weakpointcapsule (Variables|Default|GetWeakpointCapsule))
  (|CallParentFunction)
  (CallFunction|SetWeakpointCollisionKind "NewEnumerator1" 0 _weakpointcapsule)
  (CallFunction|UpdateWeakpointCollision true "NewEnumerator1" _weakpointcapsule)
  (AimAssist|RegisterAimAssistTarget 0 _weakpointcapsule))

(event EventEndPlay
  (params (EndPlayReason :type EEndPlayReason%20Enum :default "Destroyed"))
  (AimAssist|UnregisterAimAssistTarget 0))

(fn IsValidAutoFireTarget
  (params (Attacker :type Actor%20Object%20Reference))
  (return (Variables|Internal|GetbIsEnabled)))

(fn ConstructionScript
  (|CallParentFunction))
