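; Homing projectile Blueprint for the Spire boss's FirePetal MaskFire attack, built on a parent class that exposes homing hooks.
; It stores its target in ProjHomingTarget, refreshes it through FindAndSetHomingTarget in the periodic check, and returns it from GetHomingTarget.
; Max speed comes from a scalable float row evaluated with the owner as context.
; On destruction it clears both the stored target and the homing target.
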(event EventDestroyed
  (Variables|Default|SetProjHomingTarget 0)
  (Variables|Homing|SetHomingTarget 0))

(fn GetHomingTarget
  (return (Variables|Default|GetProjHomingTarget)))

(fn GetMaxSpeed
  (return (ScalableFloat|GetValueatLevelwithContextParam (Actor|GetOwner) (Variables|Default|GetProjectileMaxSpeedRow))))

(fn PeriodicCheckForHomingTarget
  (Default|FindAndSetHomingTarget))

(fn UpdateHomingTarget
  (params (HomingTarget :type Actor%20Object%20Reference))
  (Variables|Default|SetProjHomingTarget HomingTarget))

(fn ConstructionScript
  (|CallParentFunction))
