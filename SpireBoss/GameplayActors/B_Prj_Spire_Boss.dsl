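; Spire Boss projectile Blueprint that adds a sound indicator and network-dependent collision handling to its parent's behavior.
; On BeginPlay it calls the parent, then shows a sound indicator at the projectile's location using the DemonMaskFire Wandering Boss fan fire loop sound.
; On remote instances without authority it disables capsule collision.
; Overlap and Tick events are empty, and construction only calls the parent.
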(event EventBeginPlay
  (|CallParentFunction)
  (SoundIndicator|ShowSoundIndicator (Variables|Default|GetFortSoundIndicator) (Projectile|GetCurrentLocation) "/DemonMaskFire/Assets/Audio/DemonMaskFire_WanderingBoss/SC_WanderingBoss_Fan_Fire_Proj_Loop.SC_WanderingBoss_Fan_Fire_Proj_Loop" 1.0 0)
  (Networking|SwitchHasAuthority
    (:Remote
      (Collision|SetCollisionEnabled (Variables|Collision|GetCapsuleComponent) "NoCollision"))))

(event Collision|EventActorBeginOverlap
  (params (OtherActor :type Actor%20Object%20Reference)))

(event EventTick
  (params (DeltaSeconds :type Float%20%28single-precision%29)))

(fn ConstructionScript
  (|CallParentFunction))
