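# Physics asset for the Spire skeletal mesh in SpireBoss, providing capsule bodies for collision and physics simulation.
# Box001 has the largest capsule, Box002 through Box005 share a near-identical smaller, tilted capsule, and RootNode has a tiny capsule offset high along Z.
# Each Box body is constrained directly to RootNode in a star layout with uniform swing and twist limit angles, with no links between Box bodies.
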PhysicsAsset /SpireBoss/GameplayActors/Meshes/Spire_PhysicsAsset
bodies {
 Box001 [{shapeType=Capsule, length=180.1632, radius=45.3992, center=<2.927 128.9835 -46.6292>, rotation=<rot 0 -0.4745 -88.5768>}]
 Box002 [{shapeType=Capsule, length=100.5142, radius=21.4141, center=<-1.0173 56.9321 -55.0212>, rotation=<rot 0 -1.8058 -54.571>}]
 Box003 [{shapeType=Capsule, length=100.5142, radius=21.4141, center=<-1.0173 56.9321 -55.0212>, rotation=<rot 0 -1.8058 -54.571>}]
 Box004 [{shapeType=Capsule, length=100.5142, radius=21.4142, center=<-1.0173 56.9321 -55.0212>, rotation=<rot 0 -1.8058 -54.571>}]
 Box005 [{shapeType=Capsule, length=100.5142, radius=21.4141, center=<-1.0173 56.9321 -55.0212>, rotation=<rot 0 -1.8058 -54.571>}]
 RootNode [{shapeType=Capsule, length=0.505, radius=0.505, center=<0 0 442.6109>, rotation=<rot 0 -90 0>}]
}
constraints [
 {bone1Name=Box001, bone2Name=RootNode, swing1LimitDegrees=45, swing2LimitDegrees=45, twistLimitDegrees=45}
 {bone1Name=Box002, bone2Name=RootNode, swing1LimitDegrees=45, swing2LimitDegrees=45, twistLimitDegrees=45}
 {bone1Name=Box003, bone2Name=RootNode, swing1LimitDegrees=45, swing2LimitDegrees=45, twistLimitDegrees=45}
 {bone1Name=Box004, bone2Name=RootNode, swing1LimitDegrees=45, swing2LimitDegrees=45, twistLimitDegrees=45}
 {bone1Name=Box005, bone2Name=RootNode, swing1LimitDegrees=45, swing2LimitDegrees=45, twistLimitDegrees=45}
]
