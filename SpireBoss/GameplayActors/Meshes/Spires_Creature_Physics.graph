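# Physics asset for the Spires creature in the SpireBoss content, giving its shard-based skeleton collision and physics bodies.
# Single capsules cover the top, bottom, center and three large shards, their four mask pieces, and a near-zero-size small-shard master bone.
# Only the masks are constrained, each to its parent shard with matching 45-degree swing and twist limits, so the remaining bodies have no joints between them.
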PhysicsAsset /SpireBoss/GameplayActors/Meshes/Spires_Creature_Physics
bodies {
 shard_bottom [{shapeType=Capsule, length=364.8725, radius=115.9352, center=<-16.0322 -4.0405 74.2735>, rotation=<rot 1.2863 -90 -1.6005>}]
 shard_center [{shapeType=Capsule, length=358.961, radius=179.0334, center=<-34.3249 -3.305 -18.104>, rotation=<rot 0 -90 -0.111>}]
 shard_center_mask [{shapeType=Capsule, length=97.8457, radius=60.9429, center=<10.1432 -3.3711 -10.0569>, rotation=<rot 0.0918 -90 -0.515>}]
 shard_large_01 [{shapeType=Capsule, length=260.1477, radius=128.0319, center=<-16.448 6.5793 -62.4738>, rotation=<rot -1.2368 -90 14.6527>}]
 shard_large_02 [{shapeType=Capsule, length=260.1477, radius=128.0322, center=<-25.9088 6.6348 -62.5168>, rotation=<rot -1.233 -90 14.653>}]
 shard_large_03 [{shapeType=Capsule, length=260.1487, radius=128.0322, center=<-25.9335 6.4302 -62.5282>, rotation=<rot -1.2296 -90 14.6421>}]
 shard_large_mask_01 [{shapeType=Capsule, length=58.9366, radius=57.6792, center=<10.9956 6.9532 12.4514>, rotation=<rot 0 1.1859 -173.6296>}]
 shard_large_mask_02 [{shapeType=Capsule, length=58.9366, radius=57.6792, center=<10.9957 6.9532 12.4514>, rotation=<rot 0 1.1859 -173.6296>}]
 shard_large_mask_03 [{shapeType=Capsule, length=58.8646, radius=57.7883, center=<11.0212 6.8782 12.3825>, rotation=<rot 0 1.1258 -173.4733>}]
 shard_small_master [{shapeType=Capsule, length=0.505, radius=0.505, rotation=<rot 0 -90 0>}]
 shard_top [{shapeType=Capsule, length=328.1472, radius=105.4982, center=<9.8554 -5.3434 -70.9798>, rotation=<rot 1.0005 -90 1.7302>}]
}
constraints [
 {bone1Name=shard_center_mask, bone2Name=shard_center, swing1LimitDegrees=45, swing2LimitDegrees=45, twistLimitDegrees=45}
 {bone1Name=shard_large_mask_03, bone2Name=shard_large_03, swing1LimitDegrees=45, swing2LimitDegrees=45, twistLimitDegrees=45}
 {bone1Name=shard_large_mask_01, bone2Name=shard_large_01, swing1LimitDegrees=45, swing2LimitDegrees=45, twistLimitDegrees=45}
 {bone1Name=shard_large_mask_02, bone2Name=shard_large_02, swing1LimitDegrees=45, swing2LimitDegrees=45, twistLimitDegrees=45}
]
