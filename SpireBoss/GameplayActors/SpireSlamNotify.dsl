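; Particle-effect anim notify for the large spire boss's slam, sizing the spawned effect by a gameplay value.
; It calls the parent notify, then sets the "Radius" float parameter on the spawned particle system.
; The radius is twice the SF_InitialRadius scalable float, evaluated at level with the mesh owner cast to BGA_Large_Spire as context.
; If that cast fails, a fixed fallback radius is used.
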(fn Received_Notify
  (params (MeshComp :type Skeletal%20Mesh%20Component%20Object%20Reference) (Animation :type Anim%20Sequence%20Base%20Object%20Reference) (EventReference :type Anim%20Notify%20Event%20Reference%20Structure%20%28by%20ref%29))
  (bind (_asbga_large_spire _bsuccess) (Utilities|Casting|CastToBGA_Large_Spire (Components|GetOwner MeshComp)))
  (bind _returnvalue (|CallParentFunction MeshComp Animation EventReference))
  (bind _returnvalue_2 (AnimNotify|GetSpawnedEffect))
  (Effects|Components|ParticleSystem|SetFloatParameter _returnvalue_2 "Radius" (Math|Float|SelectFloat (* (ScalableFloat|GetValueatLevelwithContextParam _asbga_large_spire (Variables|Default|GetSF_InitialRadius)) 2.0) 1.2e+03 _bsuccess))
  (return _returnvalue))
